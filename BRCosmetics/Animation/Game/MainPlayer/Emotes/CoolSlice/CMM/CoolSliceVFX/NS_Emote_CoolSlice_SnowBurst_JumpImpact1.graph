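# Niagara snow burst for the CoolSlice emote, named for a jump impact, throwing mesh chunks and rotating sprite clumps.
# Both emitters spawn two burst groups at skeletal mesh sockets with mirrored sideways velocities, plus a tapering spawn rate that is zeroed in water.
# Drag slows both, gravity pulls only chunks, and a time-of-day curve tints both.
# The water splash emitter and cone velocity modules using User.VelocityMult are disabled.
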NiagaraSystem /BRCosmetics/Animation/Game/MainPlayer/Emotes/CoolSlice/CMM/CoolSliceVFX/NS_Emote_CoolSlice_SnowBurst_JumpImpact1
dependencies {
 usedDataInterfaces [&NiagaraDataInterfaceCurve, &NiagaraDataInterfaceVectorCurve]
 usedDynamicInputs [&Add_Vector]
 usedModules [
  &self:SystemSpawnScript.SystemScriptSource.SystemScriptGraph.NiagaraNodeAssignment_0.SetVariables_#1
  &SystemState
  &BlendSpaceHelper
 ]
}
emitters {
 Chunks {
  emitterUpdateScript [
   &EmitterState()
   &SpawnBurst_Instantaneous(Spawn Count=15, Use Spawn Probability=false, UseLoopCountLimit=false)
   &SpawnBurst_Instantaneous(Spawn Count=15, Spawn Group=1, Use Spawn Probability=false, UseLoopCountLimit=false) {module=SpawnBurst_Instantaneous001}
   &SpawnRate(
    SpawnRate=&MakeCustomFloatFromBool(
     Bool=$User.bIsInWater
     True Float=0
     False Float=&FloatFromCurve(
      FloatCurve=DataInterface() {propertyValues {Curve {keys [{interpMode=RCIM_Cubic, tangentMode=RCTM_User, time=0, value=1}, {interpMode=RCIM_Cubic, tangentMode=RCTM_User, time=1, value=0}]}}}
      CurveIndex=$Emitter.NormalizedLoopAge
      Scale Curve=50
     )
    )
    Use Spawn Probability=false
   )
  ]
  particleSpawnScript [
   /SetVariables(Particles.Intensity:NiagaraFloat=&FloatFromCurve(
     FloatCurve=DataInterface() {propertyValues {Curve {keys [<at 0 0>,<at 0.8 0>,<at 1 1>]}}}
     CurveIndex=&RandomRangeFloat(Minimum=0, Maximum=1)
    )) {isSetParameters=true}
   &InitializeParticle(
    Color Maximum=<rgba 0.7578 0.7578 0.9304 1>
    Color Minimum=<rgba 0.687 0.687 0.8435 1>
    Color Mode=::NewEnumerator2
    Lifetime=&Multiply_Float(
     A=&RandomRangeFloat(Minimum=0.4, Maximum=0.6)
     B=&Lerp_Float(A=0.7, B=0.9, Alpha=$Particles.Intensity)
    )
    Mass Max=2
    Mass Mode=::NewEnumerator3
    Mesh Scale=&Multiply_VectorByFloat(
     Vector=&RandomRangeVector(Minimum=<2 0.5 0.5>, Maximum=<4 3 3>)
     Float=&Lerp_Float(A=0.015, B=0.03, Alpha=$Particles.Intensity)
    )
    Mesh Scale Mode=::NewEnumerator4
    Sprite Rotation Mode=::NewEnumerator2
    UsePositionOffset=false
    Write Material Random=true
   )
   &FN_SetSpawnGroup(Number of Spawn Groups=2)
   &SkeletalMeshLocation(
    Bone / Socket Index=&SpawnGroup()
    MaskSpawnGroup=false
    Mesh Sampling Type=::NewEnumerator5
    Orientation Sampling=::NewEnumerator1
    Sampled Position Offset=&MaskVectorBySpawnGroup(
     Masked Vector=&RandomRangeVector(Minimum=<5 0 0>, Maximum=<5 -125 0>)
     Passthrough Vector=&RandomRangeVector(Minimum=<5 0 0>, Maximum=<5 -125 0>)
    )
    Skeletal Mesh=$System.Skeletal Mesh
    Socket Sampling Mode=::NewEnumerator5
   )
   &SkeletalMeshLocation(
    Bone / Socket Index=2
    Bone Sampling Mode=::NewEnumerator4
    MaskSpawnGroup=false
    Orientation Sampling=::NewEnumerator1
    Position Sampling=::NewEnumerator2
    Sampled Position Offset Coordinate Space:ENiagaraSkelMeshTransforms=::NewEnumerator2
    Skeletal Mesh=$System.Skeletal Mesh
   ) {enabled=false, module=SkeletalMeshLocation001}
   &SphereLocation(
    MaskSpawnGroup=false
    Non Uniform Scale=<0.2 1 0.1>
    Override Seed=false
    Sphere Orientation Axis:Vector3f=&ReturnMeshOrientationAxis(Mesh Orientation Axis=::XAxis, Mesh Orientation Quat=$Particles.MeshOrientation)
    Sphere Radius=75
    Surface Only=true
    Surface Only Band Thickness=1
    UseNonUniformScale=true
    UseOffset=false
    UseSphereOrientation=true
   ) {enabled=false}
   &AddVelocityInCone(
    Cone Angle=70
    Cone Axis=&RotateVectorByQuaternion(
     Vector To Rotate=&MaskVectorBySpawnGroup(Masked Vector=<0 -1 0>, Passthrough Vector=<0 -1 0>)
     Quaternion=$Particles.SkeletalMeshLocation.SampledRotation
    )
    Cone Axis Coordinate Space=::Simulation
    Module.Use Velocity Falloff On Cone Axis=true
    Velocity Distribution Along Cone Axis=0.75
    Velocity Falloff Away From Cone Axis=0.333
    Velocity Strength=&Multiply_Float(A=&RandomRangeFloat(Minimum=100, Maximum=130), B=$User.VelocityMult)
   ) {enabled=false}
   &AddVelocity(Velocity=&MaskVectorBySpawnGroup(
     Masked Vector=&RandomRangeVector(Minimum=<60 -70 70>, Maximum=<80 -100 70>)
     Passthrough Vector=&RandomRangeVector(Minimum=<-60 -70 70>, Maximum=<-80 -100 70>)
    ))
   &AddVelocity(Velocity=&Multiply_VectorByFloat(Vector=$Output.SphereLocation.SphereNormal, Float=&RandomRangeFloat(Minimum=50, Maximum=150))) {enabled=false, module=AddVelocity001}
   /SetVariables(Particles.Lifetime=&MakeCustomFloatFromBool(
     Bool=&SetBoolByFloatComparison(
      A=50
      B=&DistanceBetweenPositions(
       A=&MakeVector(
        X=0
        Y=0
        Z=&MakeFloatFromVector(Vector=Linked() {linkedVariable {name=Particles.SkeletalMeshLocation001.SampledPosition, type=NiagaraPosition}}, Channel=::NewEnumerator2)
       )
       B=&MakeVector(
        X=0
        Y=0
        Z=&MakeFloatFromVector(Vector=Linked() {linkedVariable {name=Particles.SkeletalMeshLocation.SampledPosition, type=NiagaraPosition}}, Channel=::NewEnumerator2)
       )
      )
     )
     True Float=$Particles.Lifetime
     False Float=0
    )) {enabled=false, isSetParameters=true}
  ]
  particleUpdateScript [
   &ParticleState()
   &GravityForce(Gravity=<0 0 -250>)
   &Drag(Drag=&RandomRangeFloat(Minimum=0.3, Maximum=0.5), Use Linear Drag=true, Use Rotational Drag=false)
   &BlendSpace(Rotation Contribution=0)
   &SolveForcesAndVelocity(Clamp Velocity=false, Limit Acceleration=false)
   &ScaleSpriteSize(
    Uniform Curve Scale=4
    Uniform Curve Sprite Scale=DataInterface() {propertyValues {Curve {keys [
        {arriveTangent=9.0725, interpMode=RCIM_Cubic, leaveTangent=9.0725, tangentMode=RCTM_User, time=0, value=0}
        {arriveTangent=-0.0529, interpMode=RCIM_Cubic, leaveTangent=-0.0529, tangentMode=RCTM_User, time=0.3, value=0.9847}
        {arriveTangent=-1.377, interpMode=RCIM_Cubic, leaveTangent=-1.377, tangentMode=RCTM_User, time=0.8, value=0.65}
        {arriveTangent=-0.0405, interpMode=RCIM_Cubic, leaveTangent=-0.0405, tangentMode=RCTM_User, time=1, value=0}
       ]}}}
   )
   &ScaleMeshSize(Scale Factor=&Multiply_VectorByFloat(
     Vector=<1 1 1>
     Float=&FloatFromCurve(FloatCurve=DataInterface() {propertyValues {Curve {keys [
          {arriveTangent=0.6971, interpMode=RCIM_Cubic, leaveTangent=0.6971, tangentMode=RCTM_User, time=0, value=0}
          {arriveTangent=-0.0529, interpMode=RCIM_Cubic, leaveTangent=-0.0529, tangentMode=RCTM_User, time=0.2, value=1.2}
          {arriveTangent=-1.6278, interpMode=RCIM_Cubic, leaveTangent=-1.6278, tangentMode=RCTM_User, time=0.8, value=0.65}
          {arriveTangent=-0.0405, interpMode=RCIM_Cubic, leaveTangent=-0.0405, tangentMode=RCTM_User, time=1, value=0}
         ]}}})
    ))
   &ScaleColor(
    Scale RGB=&MakeVectorFromLinearColorRGB(Color=&ColorFromCurve(
      ColorCurve=DataInterface() {propertyValues {
        AlphaCurve {keys [<at 0 1 RCIM_Cubic>]}
        BlueCurve {keys [
          <at 0.0043 0.7>
          {arriveTangent=-0.2478, interpMode=RCIM_Cubic, leaveTangent=-0.2478, time=0.1789, value=0.7}
          <at 0.2812 0.6314>
          {arriveTangent=1.3318, interpMode=RCIM_Cubic, leaveTangent=1.3318, time=0.3527, value=1}
          {arriveTangent=-1.0217, interpMode=RCIM_Cubic, leaveTangent=-1.0217, time=0.558, value=1}
          {arriveTangent=-0.9848, interpMode=RCIM_Cubic, leaveTangent=-0.9848, time=0.8029, value=0.54}
          <at 0.8615 0.7011>
          <at 1 0.7>
         ]}
        GreenCurve {keys [
          <at 0.0043 0.35>
          {arriveTangent=1.1435, interpMode=RCIM_Cubic, leaveTangent=1.1435, time=0.1789, value=0.5785}
          <at 0.2812 0.6667>
          {arriveTangent=1.2043, interpMode=RCIM_Cubic, leaveTangent=1.2043, time=0.3527, value=1}
          {arriveTangent=-0.8536, interpMode=RCIM_Cubic, leaveTangent=-0.8536, time=0.558, value=1}
          {arriveTangent=-1.3917, interpMode=RCIM_Cubic, leaveTangent=-1.3917, time=0.8029, value=0.6157}
          <at 0.8615 0.5776>
          <at 1 0.35>
         ]}
        RedCurve {keys [
          <at 0.0043 0.28>
          {arriveTangent=2.246, interpMode=RCIM_Cubic, leaveTangent=2.246, time=0.1789, value=0.5542}
          <at 0.2812 0.902>
          {arriveTangent=0.3542, interpMode=RCIM_Cubic, leaveTangent=0.3542, time=0.3527, value=1}
          {arriveTangent=-0.2221, interpMode=RCIM_Cubic, leaveTangent=-0.2221, time=0.558, value=1}
          {arriveTangent=-1.476, interpMode=RCIM_Cubic, leaveTangent=-1.476, time=0.8029, value=0.9}
          <at 0.8615 0.552>
          <at 1 0.28>
         ]}
       }}
      CurveIndex=Linked() {linkedVariable {name=NPC.FortniteNPC.RemappedTimeOfDay_0-1, type=NiagaraFloat}}
     ))
    ScaleA=false
    ScaleRGB=true
   ) {module=ScaleColor001}
  ]
  rendererClasses [&NiagaraMeshRendererProperties]
  renderers [{rendererClass=&NiagaraMeshRendererProperties, rendererIndex=0}]
 }
 Clumps {
  emitterUpdateScript [
   &EmitterState()
   &SpawnBurst_Instantaneous(Spawn Count=10, Use Spawn Probability=false, UseLoopCountLimit=false)
   &SpawnBurst_Instantaneous(Spawn Count=10, Spawn Group=1, Use Spawn Probability=false, UseLoopCountLimit=false) {module=SpawnBurst_Instantaneous001}
   &SpawnRate(
    SpawnRate=&MakeCustomFloatFromBool(
     Bool=$User.bIsInWater
     True Float=0
     False Float=&FloatFromCurve(
      FloatCurve=DataInterface() {propertyValues {Curve {keys [{interpMode=RCIM_Cubic, tangentMode=RCTM_User, time=0, value=1}, {interpMode=RCIM_Cubic, tangentMode=RCTM_User, time=1, value=0}]}}}
      CurveIndex=$Emitter.NormalizedLoopAge
      Scale Curve=50
     )
    )
    Use Spawn Probability=false
   )
  ]
  particleSpawnScript [
   /SetVariables(Particles.InitSpriteRotRate=&RandomRangeFloat(Minimum=-2, Maximum=2)) {isSetParameters=true}
   &InitializeParticle(
    Color Maximum=<rgba 0.7578 0.7578 0.9304 1>
    Color Minimum=<rgba 0.687 0.687 0.8435 1>
    Color Mode=::NewEnumerator2
    Lifetime Max=0.5
    Lifetime Min=0.4
    Lifetime Mode=::NewEnumerator1
    Position Mode=::NewEnumerator0
    Sprite Rotation Angle=&RandomRangeFloat(Minimum=0, Maximum=360)
    Sprite Rotation Mode=::NewEnumerator0
    Sprite Size Max=<15 30>
    Sprite Size Min=<10 25>
    Sprite Size Mode=::NewEnumerator2
    Sprite UV Mode=::NewEnumerator2
    Write Material Random=true
   )
   &FN_SetSpawnGroup(Number of Spawn Groups=2)
   &SkeletalMeshLocation(
    Bone / Socket Index=&SpawnGroup()
    MaskSpawnGroup=false
    Mesh Sampling Type=::NewEnumerator5
    Orientation Sampling=::NewEnumerator1
    Sampled Position Offset=&MaskVectorBySpawnGroup(
     Masked Vector=&RandomRangeVector(Minimum=<5 0 0>, Maximum=<5 -125 0>)
     Passthrough Vector=&RandomRangeVector(Minimum=<5 0 0>, Maximum=<5 -125 0>)
    )
    Skeletal Mesh=$System.Skeletal Mesh
    Socket Sampling Mode=::NewEnumerator5
   )
   &SkeletalMeshLocation(
    Bone / Socket Index=2
    Bone Sampling Mode=::NewEnumerator4
    MaskSpawnGroup=false
    Orientation Sampling=::NewEnumerator1
    Position Sampling=::NewEnumerator2
    Sampled Position Offset Coordinate Space:ENiagaraSkelMeshTransforms=::NewEnumerator2
    Skeletal Mesh=$System.Skeletal Mesh
   ) {enabled=false, module=SkeletalMeshLocation001}
   &SphereLocation(
    MaskSpawnGroup=false
    Non Uniform Scale=<0.1 1 0.1>
    Override Seed=false
    Sphere Origin:NiagaraPosition=&ReturnMeshOrientationAxis(Mesh Orientation Axis=::XAxis, Mesh Orientation Quat=$Particles.MeshOrientation)
    Sphere Radius=75
    Surface Only=true
    Surface Only Band Thickness=1
    UseNonUniformScale=true
    UseOffset=false
    UseSphereOrientation=false
   ) {enabled=false}
   &AddVelocityInCone(
    Cone Angle=10
    Cone Axis=&RotateVectorByQuaternion(
     Vector To Rotate=&MaskVectorBySpawnGroup(Masked Vector=<0 -1 0>, Passthrough Vector=<0 -1 0>)
     Quaternion=$Particles.SkeletalMeshLocation.SampledRotation
    )
    Cone Axis Coordinate Space=::Simulation
    Module.Use Velocity Falloff On Cone Axis=true
    Velocity Distribution Along Cone Axis=0.75
    Velocity Falloff Away From Cone Axis=0.333
    Velocity Strength=&Multiply_Float(A=&RandomRangeFloat(Minimum=150, Maximum=180), B=$User.VelocityMult)
   ) {enabled=false}
   &AddVelocity(Velocity=&MaskVectorBySpawnGroup(
     Masked Vector=&RandomRangeVector(Minimum=<50 -70 30>, Maximum=<150 -100 70>)
     Passthrough Vector=&RandomRangeVector(Minimum=<-50 -70 30>, Maximum=<-150 -100 70>)
    ))
   /SetVariables(Particles.Lifetime=&MakeCustomFloatFromBool(
     Bool=&SetBoolByFloatComparison(
      A=50
      B=&DistanceBetweenPositions(
       A=&MakeVector(
        X=0
        Y=0
        Z=&MakeFloatFromVector(Vector=Linked() {linkedVariable {name=Particles.SkeletalMeshLocation001.SampledPosition, type=NiagaraPosition}}, Channel=::NewEnumerator2)
       )
       B=&MakeVector(
        X=0
        Y=0
        Z=&MakeFloatFromVector(Vector=Linked() {linkedVariable {name=Particles.SkeletalMeshLocation.SampledPosition, type=NiagaraPosition}}, Channel=::NewEnumerator2)
       )
      )
     )
     True Float=$Particles.Lifetime
     False Float=0
    )) {enabled=false, isSetParameters=true}
  ]
  particleUpdateScript [
   &ParticleState()
   &Drag(Drag=&RandomRangeFloat(Minimum=0.8, Maximum=1.2), Use Linear Drag=true, Use Rotational Drag=false)
   &BlendSpace(Rotation Contribution=0)
   &SolveForcesAndVelocity(Clamp Velocity=false, Limit Acceleration=false)
   &SpriteRotationRate(Rotation Rate=&Multiply_Float(
     A=$Particles.InitSpriteRotRate
     B=&FloatFromCurve(
      FloatCurve=DataInterface() {propertyValues {Curve {keys [
          {arriveTangent=-4.0107, interpMode=RCIM_Cubic, leaveTangent=-4.0107, tangentMode=RCTM_User, time=-0.0046, value=1.0113}
          {arriveTangent=-0.2496, interpMode=RCIM_Cubic, leaveTangent=-0.2496, tangentMode=RCTM_User, time=0.5, value=0.16}
         ]}}}
      CurveIndex=$Emitter.Age
      Scale Curve=150
     )
    ))
   &ScaleSpriteSize(Scale Factor=&Vector2DFromFloat(Value=&FloatFromCurve(FloatCurve=DataInterface() {propertyValues {Curve {keys [
          {arriveTangent=0.0905, interpMode=RCIM_Cubic, leaveTangent=0.0905, tangentMode=RCTM_User, time=0, value=0.5}
          {arriveTangent=5.1283, interpMode=RCIM_Cubic, leaveTangent=5.1283, tangentMode=RCTM_User, time=0.3, value=1.125}
          {arriveTangent=0.3828, interpMode=RCIM_Cubic, leaveTangent=0.3828, tangentMode=RCTM_User, time=0.8, value=2}
         ]}}})))
   &ScaleColor(
    Scale RGB=&MakeVectorFromLinearColorRGB(Color=&ColorFromCurve(
      ColorCurve=DataInterface() {propertyValues {
        AlphaCurve {keys [<at 0 1 RCIM_Cubic>]}
        BlueCurve {keys [
          <at 0.0043 0.7>
          {arriveTangent=-0.2478, interpMode=RCIM_Cubic, leaveTangent=-0.2478, time=0.1789, value=0.7}
          <at 0.2812 0.6314>
          {arriveTangent=1.3318, interpMode=RCIM_Cubic, leaveTangent=1.3318, time=0.3527, value=1}
          {arriveTangent=-1.0217, interpMode=RCIM_Cubic, leaveTangent=-1.0217, time=0.558, value=1}
          {arriveTangent=-0.9848, interpMode=RCIM_Cubic, leaveTangent=-0.9848, time=0.8029, value=0.54}
          <at 0.8615 0.7011>
          <at 1 0.7>
         ]}
        GreenCurve {keys [
          <at 0.0043 0.35>
          {arriveTangent=1.1435, interpMode=RCIM_Cubic, leaveTangent=1.1435, time=0.1789, value=0.5785}
          <at 0.2812 0.6667>
          {arriveTangent=1.2043, interpMode=RCIM_Cubic, leaveTangent=1.2043, time=0.3527, value=1}
          {arriveTangent=-0.8536, interpMode=RCIM_Cubic, leaveTangent=-0.8536, time=0.558, value=1}
          {arriveTangent=-1.3917, interpMode=RCIM_Cubic, leaveTangent=-1.3917, time=0.8029, value=0.6157}
          <at 0.8615 0.5776>
          <at 1 0.35>
         ]}
        RedCurve {keys [
          <at 0.0043 0.28>
          {arriveTangent=2.246, interpMode=RCIM_Cubic, leaveTangent=2.246, time=0.1789, value=0.5542}
          <at 0.2812 0.902>
          {arriveTangent=0.3542, interpMode=RCIM_Cubic, leaveTangent=0.3542, time=0.3527, value=1}
          {arriveTangent=-0.2221, interpMode=RCIM_Cubic, leaveTangent=-0.2221, time=0.558, value=1}
          {arriveTangent=-1.476, interpMode=RCIM_Cubic, leaveTangent=-1.476, time=0.8029, value=0.9}
          <at 0.8615 0.552>
          <at 1 0.28>
         ]}
       }}
      CurveIndex=Linked() {linkedVariable {name=NPC.FortniteNPC.RemappedTimeOfDay_0-1, type=NiagaraFloat}}
     ))
    ScaleA=false
    ScaleRGB=true
   ) {module=ScaleColor001}
   &DynamicMaterialParameters(
    Index 0 Param 1=$Particles.NormalizedAge
    Module.Index0Float0Write=true
    Module.Index0Float1Write=true
    Module.Index0Float2Write=true
    Module.Index0Float3Write=true
   )
  ]
  rendererClasses [&NiagaraSpriteRendererProperties]
  renderers [
   {rendererClass=&NiagaraSpriteRendererProperties}
   {rendererClass=&NiagaraSpriteRendererProperties, rendererIndex=1}
  ]
 }
 Water_Plasma_Splash001 {
  bEnabled=false
  emitterUpdateScript [
   &EmitterState(Life Cycle Mode=::NewEnumerator1, Loop Behavior=::NewEnumerator1, Loop Delay=$System.Delay, Loop Duration=0.25, UseLoopDelay=false)
   &SpawnBurst_Instantaneous(Spawn Count=3, Use Spawn Probability=false, UseLoopCountLimit=false)
   &SpawnBurst_Instantaneous(Spawn Count=3, Spawn Group=1, Use Spawn Probability=false, UseLoopCountLimit=false) {module=SpawnBurst_Instantaneous001}
  ]
  particleSpawnScript [
   &FN_SetSpawnGroup(Number of Spawn Groups=2) {enabled=false}
   &InitializeParticle(
    Color=<rgba 1.5 1.5 1.5 0.8>
    Lifetime=&RandomRangeFloat(Minimum=0.3, Maximum=0.4)
    Mesh Scale Max=<0.55 0.55 0.8>
    Mesh Scale Min=<0.5 0.5 0.7>
    Mesh Scale Mode=::NewEnumerator2
    Position Offset=&MakeVector(X=0, Y=0, Z=0)
    UsePositionOffset=true
    Write Material Random=true
   )
   &DynamicMaterialParameters(
    Index 0 Param 1=&RandomRangeFloat(Minimum=0, Maximum=1)
    Module.Index0Param0Write=true
    Module.Index0Param1Write=true
    Module.Index0Param2Write=true
    Module.Index0Param3Write=true
   )
   &AddVelocity(
    Module.Use Falloff Distance=false
    Origin Offset=<0 0 -2>
    Override Seed=false
    Velocity Mode=::NewEnumerator1
    Velocity Speed=&RandomRangeFloat(Minimum=10, Maximum=20)
   ) {enabled=false, module=AddVelocity001}
   &SkeletalMeshLocation(
    Bone / Socket Index=2
    Bone Sampling Mode=::NewEnumerator4
    MaskSpawnGroup=false
    Orientation Sampling=::NewEnumerator1
    Position Sampling=::NewEnumerator2
    Sampled Position Offset Coordinate Space:ENiagaraSkelMeshTransforms=::NewEnumerator2
    Skeletal Mesh=$System.Skeletal Mesh
   ) {module=SkeletalMeshLocation001}
   &SkeletalMeshLocation(
    Bone / Socket Index=$Particles.SpawnGroup
    MaskSpawnGroup=false
    Mesh Sampling Type=::NewEnumerator5
    Sampled Position Offset=&MakeVector(X=0, Y=&MaskFloatBySpawnGroup(Masked Float=5, Passthrough Float=-5), Z=0)
    Skeletal Mesh=$System.Skeletal Mesh
    Socket Sampling Mode=::NewEnumerator5
   )
   /SetVariables(Particles.Lifetime=&MakeCustomFloatFromBool(
     Bool=&SetBoolByFloatComparison(
      A=50
      B=&DistanceBetweenPositions(
       A=&MakeVector(
        X=0
        Y=0
        Z=&MakeFloatFromVector(Vector=Linked() {linkedVariable {name=Particles.SkeletalMeshLocation001.SampledPosition, type=NiagaraPosition}}, Channel=::NewEnumerator2)
       )
       B=&MakeVector(
        X=0
        Y=0
        Z=&MakeFloatFromVector(Vector=Linked() {linkedVariable {name=Particles.SkeletalMeshLocation.SampledPosition, type=NiagaraPosition}}, Channel=::NewEnumerator2)
       )
      )
     )
     True Float=$Particles.Lifetime
     False Float=0
    )) {isSetParameters=true}
   &InitialMeshOrientation(Mesh Orientation Mode=::NewEnumerator7, Use Rotation Vector=true) {enabled=false}
   &AddVelocity(
    Velocity=&ReturnMeshOrientationAxis(Mesh Orientation Axis=::XAxis, Mesh Orientation Quat=$Particles.MeshOrientation)
    Velocity Speed Scale=&RandomRangeFloat(Minimum=200, Maximum=250)
   ) {module=AddVelocity002}
   /SetVariables(Particles.Scale=&Multiply_VectorByFloat(
     Vector=$Particles.Scale
     Float=&FloatFromCurve(
      FloatCurve=DataInterface() {propertyValues {Curve {keys [<at 0 0>,<at 0.5 0>,<at 1 1>]}}}
      CurveIndex=&NormalizeDistanceRange(
       Start Position:NiagaraPosition=&MakeVector(
        X=0
        Y=&MakeFloatFromVector(Vector=Linked() {linkedVariable {name=Particles.SkeletalMeshLocation.SampledPosition, type=NiagaraPosition}}, Channel=::NewEnumerator1)
        Z=0
       )
       End Position:NiagaraPosition=&MakeVector(
        X=0
        Y=&MakeFloatFromVector(Vector=Linked() {linkedVariable {name=Particles.SkeletalMeshLocation001.SampledPosition, type=NiagaraPosition}}, Channel=::NewEnumerator1)
        Z=0
       )
       Distance:NiagaraFloat=50
       Invert Normalized Range:NiagaraBool=false
      )
     )
    )) {enabled=false, isSetParameters=true}
  ]
  particleUpdateScript [
   &ParticleState()
   &GravityForce(Gravity=<0 0 -10>)
   &Drag(Drag=2, Use Linear Drag=true, Use Rotational Drag=true)
   &ScaleColor(
    Scale Alpha=&FloatFromCurve(
     FloatCurve=DataInterface() {propertyValues {Curve {keys [
         {arriveTangent=9.8898, interpMode=RCIM_Cubic, leaveTangent=9.8898, tangentMode=RCTM_User, time=0, value=0}
         {interpMode=RCIM_Cubic, tangentMode=RCTM_User, time=0.3, value=1}
         {arriveTangent=-3.6098, tangentMode=RCTM_User, time=1, value=0}
        ]}}}
     Scale Curve=2
    )
    Scale RGB=&VectorFromFloat(Value=&FloatFromCurve(
      FloatCurve=DataInterface() {propertyValues {Curve {keys [
          {arriveTangent=9.0725, interpMode=RCIM_Cubic, leaveTangent=9.0725, tangentMode=RCTM_User, time=0, value=0.5}
          {arriveTangent=-0.7694, interpMode=RCIM_Cubic, leaveTangent=-0.7694, tangentMode=RCTM_User, time=0.1408, value=0.9847}
          {arriveTangent=-0.0405, interpMode=RCIM_Cubic, leaveTangent=-0.0405, tangentMode=RCTM_User, time=1, value=0.5}
         ]}}}
      Scale Curve=&FNTimeOfDayScalar(Night Scale=0.5, Morning Scale=1, Evening Scale=1)
     ))
    ScaleA=true
    ScaleRGB=true
   ) {enabled=false}
   &InterpolateOverTime(
    Rate Of Change=10
    Target Vector Value:Vector3f=&AddVectorToPosition(
     Position=$Particles.Position
     Vector=&RotateVectorByQuaternion(Vector To Rotate=<0 -200 0>, Quaternion=$Particles.SkeletalMeshLocation.SampledRotation)
    )
    Variable Type=::NewEnumerator2
   ) {enabled=false}
   &DynamicMaterialParameters(
    Index 0 Param 1=&FloatFromCurve(FloatCurve=DataInterface() {propertyValues {Curve {defaultValue=1, keys [{tangentMode=RCTM_User, time=0, value=0}, <at 0.499 0.3204>, {tangentMode=RCTM_User, time=1, value=1}]}}})
    Index 0 Param 3=0.5393
    Index 2 Param 1=&RandomRangeFloat(Minimum=1, Maximum=1.5)
    Index 2 Param 2=&RandomRangeFloat(Minimum=1, Maximum=1.5)
    Index 2 Param 3=-3.5
    Index 2 Param 4=&RandomRangeFloat(Minimum=3, Maximum=4)
    Index 3 Param 1=&RandomRangeFloat(Minimum=0.2, Maximum=0.3)
    Index 3 Param 2=&RandomRangeFloat(Minimum=0, Maximum=1)
    Index 3 Param 3=&RandomRangeFloat(Minimum=0, Maximum=1)
    Module.Index0Float0Write=true
    Module.Index0Float1Write=true
    Module.Index0Float2Write=true
    Module.Index0Float3Write=true
    Module.Index1Float0Write=true
    Module.Index1Float1Write=true
    Module.Index1Float2Write=true
    Module.Index1Float3Write=true
    Module.Index2Float0Write=true
    Module.Index2Float1Write=true
    Module.Index2Float2Write=true
    Module.Index2Float3Write=true
    Module.Index3Float0Write=true
    Module.Index3Float1Write=true
    Module.Index3Float2Write=true
    Module.Index3Float3Write=true
    Write Parameter Index 1=::NewEnumerator1
    Write Parameter Index 2=::NewEnumerator1
    Write Parameter Index 3=::NewEnumerator1
   ) {module=DynamicMaterialParameters001}
   &ScaleMeshSize(Scale Factor=&VectorFromCurve(
     VectorCurve=DataInterface() {propertyValues {
       XCurve {keys [<at 0 0>,<at 0.496 0.8157>,<at 1 1>]}
       YCurve {keys [<at 0 0>,<at 0.487 0.8157>,<at 1 1>]}
       ZCurve {keys [<at 0 0>,<at 0.4825 0.8628>,<at 1 1>]}
      }}
     Scale Curve=&MaskVectorBySpawnGroup(Masked Vector=<-1 1 1>, Passthrough Vector=<1 1 1>)
    ))
   &SolveForcesAndVelocity(Clamp Velocity=false, Limit Acceleration=false)
  ]
  rendererClasses [&NiagaraMeshRendererProperties]
  renderers [{rendererClass=&NiagaraMeshRendererProperties, rendererIndex=0}]
 }
}
summary {
 systemName=NS_Emote_CoolSlice_SnowBurst_JumpImpact1
 emitters [
  {bEnabled=false, emitterName=Water_Plasma_Splash001, rendererClasses [&NiagaraMeshRendererProperties]}
  {emitterName=Clumps, rendererClasses [&NiagaraSpriteRendererProperties]}
  {emitterName=Chunks, rendererClasses [&NiagaraMeshRendererProperties]}
 ]
 userVariables [
  User.bIsInWater {type=NiagaraBool}
  User.SkeletalMesh {
   type=NiagaraDataInterfaceSkeletalMesh
   defaultValue {
    struct=NiagaraExt_VariableValue_DataInterface
    value {dataInterface=&self:NiagaraDataInterfaceSkeletalMesh_0}
   }
  }
  User.SmokeColor {
   type=NiagaraDataInterfaceColorCurve
   description=""
   defaultValue {
    struct=NiagaraExt_VariableValue_DataInterface
    value {
     dataInterface=&self:NiagaraDataInterfaceColorCurve_1
     dataInterfaceClass=&NiagaraDataInterfaceColorCurve
    }
   }
  }
  User.SnowColor {
   type=NiagaraDataInterfaceColorCurve
   description=""
   defaultValue {
    struct=NiagaraExt_VariableValue_DataInterface
    value {
     dataInterface=&self:NiagaraDataInterfaceColorCurve_0
     dataInterfaceClass=&NiagaraDataInterfaceColorCurve
    }
   }
  }
  User.SnowColorMax {type=Vector3f, description="", defaultValue {struct=Vector3f, value=<1 1 1>}}
  User.SnowColorMin {type=Vector3f, description="", defaultValue {struct=Vector3f, value=<1 1 1>}}
  User.VelocityMult {type=NiagaraFloat, defaultValue=2, description=""}
 ]
}
