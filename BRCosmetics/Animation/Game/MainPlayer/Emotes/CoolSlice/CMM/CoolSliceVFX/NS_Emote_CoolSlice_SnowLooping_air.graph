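# Niagara system for the looping snow spray of the CoolSlice player emote, named as the air variant.
# Mesh Chunks and sprite Clumps spawn at offsets from two skeletal mesh bones or sockets, launched along bone-oriented cones scaled by User.VelocityMult.
# Both stop spawning when User.bIsInWater is set and tint by a time-of-day parameter collection curve.
# Several other emitters, including dust spikes, GPU flakes and duplicate chunk and clump layers, are disabled.
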NiagaraSystem /BRCosmetics/Animation/Game/MainPlayer/Emotes/CoolSlice/CMM/CoolSliceVFX/NS_Emote_CoolSlice_SnowLooping_air
dependencies {
 usedDataInterfaces [
  &NiagaraDataInterfaceSkeletalMesh
  &NiagaraDataInterfaceCurve
  &NiagaraDataInterfaceArrayFloat
  &NiagaraDataInterfaceSpriteRendererInfo
  &NiagaraDataInterfaceVectorCurve
 ]
 usedDynamicInputs [&Add_Vector]
 usedModules [
  &self:SystemSpawnScript.SystemScriptSource.SystemScriptGraph.NiagaraNodeAssignment_4.SetVariables_#1
  &SystemState
  &BlendSpaceHelper
 ]
}
emitters {
 Chunks {
  emitterUpdateScript [
   &EmitterState(Life Cycle Mode=::NewEnumerator1, Loop Duration=0.9, UseLoopDelay=false)
   &SpawnRate(
    SpawnRate=&MakeCustomFloatFromBool(Bool=$User.bIsInWater, True Float=0, False Float=60)
    Use Spawn Probability=false
   )
  ]
  particleSpawnScript [
   /SetVariables(Particles.Intensity:NiagaraFloat=&FloatFromCurve(
     FloatCurve=DataInterface() {propertyValues {Curve {keys [<at 0 0>,<at 0.8 0>,<at 1 1>]}}}
     CurveIndex=&RandomRangeFloat(Minimum=0, Maximum=1)
    )) {isSetParameters=true, module=SetVariables_#2}
   &InitializeParticle(
    Color Maximum=<rgba 0.7578 0.7578 0.9304 1>
    Color Minimum=<rgba 0.687 0.687 0.8435 1>
    Color Mode=::NewEnumerator2
    Lifetime=&Multiply_Float(
     A=&RandomRangeFloat(Minimum=0.35, Maximum=0.45)
     B=&Lerp_Float(A=0.7, B=0.9, Alpha=$Particles.Intensity)
    )
    Mass Max=2
    Mass Mode=::NewEnumerator3
    Mesh Scale=&Multiply_VectorByFloat(
     Vector=&RandomRangeVector(Minimum=<2 0.5 0.5>, Maximum=<4 3 3>)
     Float=&Lerp_Float(A=0.015, B=0.03, Alpha=$Particles.Intensity)
    )
    Mesh Scale Mode=::NewEnumerator4
    Sprite Rotation Mode=::NewEnumerator2
    UsePositionOffset=false
    Write Material Random=true
   )
   &FN_SetSpawnGroup(Number of Spawn Groups=2)
   &SkeletalMeshLocation(
    Bone / Socket Index=$Particles.SpawnGroup
    Bone Sampling Mode=::NewEnumerator4
    MaskSpawnGroup=false
    Orientation Sampling=::NewEnumerator1
    Sampled Position Offset=&MakeVector(
     X=&MaskFloatBySpawnGroup(
      Masked Float=&SelectFloatFromArray(
       Array Sampling Mode=::NewEnumerator0
       Direct Array Index=?
       Float Selection Array=DataInterface() {propertyValues {FloatData [-4,14]}}
       Random Seed=0
      )
      Passthrough Float=&SelectFloatFromArray(
       Array Sampling Mode=::NewEnumerator0
       Direct Array Index=?
       Float Selection Array=DataInterface() {propertyValues {FloatData [4,-14]}}
       Random Seed=0
      )
     )
     Y=&MakeFloatFromVector(Vector=&RandomRangeVector(Minimum=<0 -45 0>, Maximum=<0 -65 0>), Channel=::NewEnumerator1)
     Z=0
    )
    Skeletal Mesh=$System.Skeletal Mesh
   )
   &SkeletalMeshLocation(
    Bone / Socket Index=2
    Bone Sampling Mode=::NewEnumerator4
    MaskSpawnGroup=false
    Orientation Sampling=::NewEnumerator1
    Position Sampling=::NewEnumerator2
    Sampled Position Offset Coordinate Space:ENiagaraSkelMeshTransforms=::NewEnumerator2
    Skeletal Mesh=$System.Skeletal Mesh
   ) {module=SkeletalMeshLocation001}
   &SphereLocation(
    MaskSpawnGroup=false
    Override Seed=false
    Sphere Coordinate Space=::Simulation
    Sphere Radius=8
    Surface Only=true
    Surface Only Band Thickness=1
    UseNonUniformScale=false
    UseOffset=false
    UseSphereOrientation=false
   )
   &AddVelocityInCone(
    Cone Angle=70
    Cone Axis=&RotateVectorByQuaternion(
     Vector To Rotate=&MaskVectorBySpawnGroup(Masked Vector=<0 -1 0>, Passthrough Vector=<0 -1 0>)
     Quaternion=$Particles.SkeletalMeshLocation.SampledRotation
    )
    Cone Axis Coordinate Space=::Simulation
    Module.Use Velocity Falloff On Cone Axis=true
    Velocity Distribution Along Cone Axis=0.75
    Velocity Falloff Away From Cone Axis=0.333
    Velocity Strength=&Multiply_Float(A=&RandomRangeFloat(Minimum=180, Maximum=200), B=$User.VelocityMult)
   )
   &AddVelocity(Velocity=&RandomRangeVector(Minimum=<0 0 70>, Maximum=<0 0 100>))
   /SetVariables(Particles.Lifetime=&MakeCustomFloatFromBool(
     Bool=&SetBoolByFloatComparison(
      A=75
      B=&DistanceBetweenPositions(
       A=&MakeVector(
        X=0
        Y=0
        Z=&MakeFloatFromVector(Vector=Linked() {linkedVariable {name=Particles.SkeletalMeshLocation001.SampledPosition, type=NiagaraPosition}}, Channel=::NewEnumerator2)
       )
       B=&MakeVector(
        X=0
        Y=0
        Z=&MakeFloatFromVector(Vector=Linked() {linkedVariable {name=Particles.SkeletalMeshLocation.SampledPosition, type=NiagaraPosition}}, Channel=::NewEnumerator2)
       )
      )
     )
     True Float=$Particles.Lifetime
     False Float=0
    )) {enabled=false, isSetParameters=true, module=SetVariables_#3}
  ]
  particleUpdateScript [
   &ParticleState()
   &Drag(Drag=&RandomRangeFloat(Minimum=0.3, Maximum=0.5), Use Linear Drag=true, Use Rotational Drag=false)
   &GravityForce(Gravity=<0 0 -700>)
   &MeshRotationForce(Rotation=&RandomRangeVector(Minimum=<0 0 0>, Maximum=<-50 1 1>))
   &BlendSpace(Rotation Contribution=0)
   &SolveForcesAndVelocity(Clamp Velocity=false, Limit Acceleration=false)
   &ScaleSpriteSize(
    Uniform Curve Scale=4
    Uniform Curve Sprite Scale=DataInterface() {propertyValues {Curve {keys [
        {arriveTangent=9.0725, interpMode=RCIM_Cubic, leaveTangent=9.0725, tangentMode=RCTM_User, time=0, value=0}
        {arriveTangent=-0.0529, interpMode=RCIM_Cubic, leaveTangent=-0.0529, tangentMode=RCTM_User, time=0.3, value=0.9847}
        {arriveTangent=-1.377, interpMode=RCIM_Cubic, leaveTangent=-1.377, tangentMode=RCTM_User, time=0.8, value=0.65}
        {arriveTangent=-0.0405, interpMode=RCIM_Cubic, leaveTangent=-0.0405, tangentMode=RCTM_User, time=1, value=0}
       ]}}}
   )
   &ScaleColor(
    Scale RGB=&MakeVectorFromLinearColorRGB(Color=&ColorFromCurve(
      ColorCurve=DataInterface() {propertyValues {
        AlphaCurve {keys [<at 0 1 RCIM_Cubic>]}
        BlueCurve {keys [
          <at 0.0043 0.7>
          {arriveTangent=-0.2478, interpMode=RCIM_Cubic, leaveTangent=-0.2478, time=0.1789, value=0.7}
          <at 0.2812 0.6314>
          {arriveTangent=1.3318, interpMode=RCIM_Cubic, leaveTangent=1.3318, time=0.3527, value=1}
          {arriveTangent=-1.0217, interpMode=RCIM_Cubic, leaveTangent=-1.0217, time=0.558, value=1}
          {arriveTangent=-0.9848, interpMode=RCIM_Cubic, leaveTangent=-0.9848, time=0.8029, value=0.54}
          <at 0.8615 0.7011>
          <at 1 0.7>
         ]}
        GreenCurve {keys [
          <at 0.0043 0.35>
          {arriveTangent=1.1435, interpMode=RCIM_Cubic, leaveTangent=1.1435, time=0.1789, value=0.5785}
          <at 0.2812 0.6667>
          {arriveTangent=1.2043, interpMode=RCIM_Cubic, leaveTangent=1.2043, time=0.3527, value=1}
          {arriveTangent=-0.8536, interpMode=RCIM_Cubic, leaveTangent=-0.8536, time=0.558, value=1}
          {arriveTangent=-1.3917, interpMode=RCIM_Cubic, leaveTangent=-1.3917, time=0.8029, value=0.6157}
          <at 0.8615 0.5776>
          <at 1 0.35>
         ]}
        RedCurve {keys [
          <at 0.0043 0.28>
          {arriveTangent=2.246, interpMode=RCIM_Cubic, leaveTangent=2.246, time=0.1789, value=0.5542}
          <at 0.2812 0.902>
          {arriveTangent=0.3542, interpMode=RCIM_Cubic, leaveTangent=0.3542, time=0.3527, value=1}
          {arriveTangent=-0.2221, interpMode=RCIM_Cubic, leaveTangent=-0.2221, time=0.558, value=1}
          {arriveTangent=-1.476, interpMode=RCIM_Cubic, leaveTangent=-1.476, time=0.8029, value=0.9}
          <at 0.8615 0.552>
          <at 1 0.28>
         ]}
       }}
      CurveIndex=Linked() {linkedVariable {name=NPC.FortniteNPC.RemappedTimeOfDay_0-1, type=NiagaraFloat}}
     ))
    ScaleA=false
    ScaleRGB=true
   ) {module=ScaleColor001}
   &ScaleMeshSize(Scale Factor=&Multiply_VectorByFloat(
     Vector=<1 1 1>
     Float=&FloatFromCurve(FloatCurve=DataInterface() {propertyValues {Curve {keys [
          {arriveTangent=0.6971, interpMode=RCIM_Cubic, leaveTangent=0.6971, tangentMode=RCTM_User, time=0, value=0}
          {arriveTangent=-0.0529, interpMode=RCIM_Cubic, leaveTangent=-0.0529, tangentMode=RCTM_User, time=0.2, value=1.2}
          {arriveTangent=-1.6278, interpMode=RCIM_Cubic, leaveTangent=-1.6278, tangentMode=RCTM_User, time=0.8, value=0.65}
          {arriveTangent=-0.0405, interpMode=RCIM_Cubic, leaveTangent=-0.0405, tangentMode=RCTM_User, time=1, value=0}
         ]}}})
    ))
  ]
  rendererClasses [&NiagaraMeshRendererProperties]
  renderers [{rendererClass=&NiagaraMeshRendererProperties, rendererIndex=0}]
 }
 Chunks001 {
  bEnabled=false
  emitterUpdateScript [
   &EmitterState()
   &SpawnRate(
    SpawnGroup=3
    SpawnRate=&FloatFromCurve(
     FloatCurve=DataInterface() {propertyValues {Curve {keys [<at 0.1 0 RCIM_Constant>,<at 0.2106 1 RCIM_Constant>,<at 0.2212 0 RCIM_Constant>,<at 0.2936 1 RCIM_Constant>,<at 0.3031 0 RCIM_Constant>,<at 0.3401 1 RCIM_Constant>,<at 0.3465 0 RCIM_Constant>,<at 0.381 1 RCIM_Constant>,<at 0.3884 0 RCIM_Constant>,<at 0.4331 1 RCIM_Constant>,<at 0.4399 0 RCIM_Constant>,<at 0.5345 1 RCIM_Constant>,<at 0.5584 1 RCIM_Constant>,<at 0.5636 0 RCIM_Constant>,<at 0.7232 1 RCIM_Constant>,<at 0.7369 0 RCIM_Constant>,<at 0.8612 1 RCIM_Constant>,<at 0.888 1 RCIM_Constant>,<at 0.8943 0 RCIM_Constant>]}}}
     CurveIndex=$Emitter.NormalizedLoopAge
     Scale Curve=20
    )
    Use Spawn Probability=false
   )
   &SpawnRate(
    SpawnGroup=4
    SpawnRate=&FloatFromCurve(
     FloatCurve=DataInterface() {propertyValues {Curve {keys [<at 0.1 0 RCIM_Constant>,<at 0.2106 1 RCIM_Constant>,<at 0.2212 0 RCIM_Constant>,<at 0.2936 1 RCIM_Constant>,<at 0.3031 0 RCIM_Constant>,<at 0.3401 1 RCIM_Constant>,<at 0.3465 0 RCIM_Constant>,<at 0.381 1 RCIM_Constant>,<at 0.3884 0 RCIM_Constant>,<at 0.4331 1 RCIM_Constant>,<at 0.4399 0 RCIM_Constant>,<at 0.5345 1 RCIM_Constant>,<at 0.5584 1 RCIM_Constant>,<at 0.5636 0 RCIM_Constant>,<at 0.7232 1 RCIM_Constant>,<at 0.7369 0 RCIM_Constant>,<at 0.8612 1 RCIM_Constant>,<at 0.888 1 RCIM_Constant>,<at 0.8943 0 RCIM_Constant>]}}}
     CurveIndex=$Emitter.NormalizedLoopAge
     Scale Curve=20
    )
    Use Spawn Probability=false
   ) {module=SpawnRate001}
  ]
  particleSpawnScript [
   /SetVariables(Particles.Intensity:NiagaraFloat=&FloatFromCurve(
     FloatCurve=DataInterface() {propertyValues {Curve {keys [<at 0 0>,<at 0.8 0>,<at 1 1>]}}}
     CurveIndex=&RandomRangeFloat(Minimum=0, Maximum=1)
    )) {isSetParameters=true, module=SetVariables_#2}
   &InitializeParticle(
    Lifetime=&Multiply_Float(
     A=&RandomRangeFloat(Minimum=0.2, Maximum=0.3)
     B=&Lerp_Float(A=0.7, B=0.9, Alpha=$Particles.Intensity)
    )
    Mass Max=2
    Mass Mode=::NewEnumerator3
    Mesh Scale=&Multiply_VectorByFloat(
     Vector=&RandomRangeVector(Minimum=<2 0.5 0.5>, Maximum=<4 2 2>)
     Float=&Lerp_Float(A=0.02, B=0.03, Alpha=$Particles.Intensity)
    )
    Mesh Scale Mode=::NewEnumerator4
    Sprite Rotation Mode=::NewEnumerator2
    UsePositionOffset=false
    Write Material Random=true
   )
   &FN_SetSpawnGroup(Number of Spawn Groups=2) {enabled=false}
   &SkeletalMeshLocation(
    Bone / Socket Index=&SpawnGroup()
    Bone Sampling Mode=::NewEnumerator4
    MaskSpawnGroup=false
    Skeletal Mesh=$System.Skeletal Mesh
   )
   &SphereLocation(
    MaskSpawnGroup=false
    Override Seed=false
    Sphere Coordinate Space=::Simulation
    Sphere Origin:NiagaraPosition=$Particles.SkeletalMeshLocation.SampledPosition
    Sphere Radius=8
    Surface Only=true
    Surface Only Band Thickness=1
    UseNonUniformScale=false
    UseOffset=false
    UseSphereOrientation=false
   )
   &AddVelocityInCone(
    Cone Angle=70
    Cone Axis=&RotateVectorByQuaternion(
     Vector To Rotate=&MaskVectorBySpawnGroup(Masked Vector=<0 -1 0>, Passthrough Vector=<0 -1 0>)
     Quaternion=$Particles.SkeletalMeshLocation.SampledRotation
    )
    Cone Axis Coordinate Space=::Simulation
    Module.Use Velocity Falloff On Cone Axis=true
    Velocity Distribution Along Cone Axis=0.75
    Velocity Falloff Away From Cone Axis=0.333
    Velocity Strength=&Multiply_Float(A=&RandomRangeFloat(Minimum=180, Maximum=200), B=$User.VelocityMult)
   ) {enabled=false}
   &AddVelocity(Velocity=&Multiply_VectorByFloat(Vector=$Output.SphereLocation.SphereNormal, Float=&RandomRangeFloat(Minimum=200, Maximum=300))) {module=AddVelocity001}
   &AddVelocity(Velocity=&RandomRangeVector(Minimum=<0 0 70>, Maximum=<0 0 100>))
   /SetVariables(Particles.Lifetime=&MakeCustomFloatFromBool(
     Bool=&SetBoolByFloatComparison(
      A=50
      B=&DistanceBetweenPositions(
       A=&MakeVector(
        X=0
        Y=0
        Z=&MakeFloatFromVector(Vector=Linked() {linkedVariable {name=Particles.SkeletalMeshLocation001.SampledPosition, type=NiagaraPosition}}, Channel=::NewEnumerator2)
       )
       B=&MakeVector(
        X=0
        Y=0
        Z=&MakeFloatFromVector(Vector=Linked() {linkedVariable {name=Particles.SkeletalMeshLocation.SampledPosition, type=NiagaraPosition}}, Channel=::NewEnumerator2)
       )
      )
     )
     True Float=$Particles.Lifetime
     False Float=0
    )) {enabled=false, isSetParameters=true, module=SetVariables_#3}
  ]
  particleUpdateScript [
   &ParticleState()
   &GravityForce(Gravity=<0 0 -250>)
   &Drag(Drag=&RandomRangeFloat(Minimum=0.3, Maximum=0.5), Use Linear Drag=true, Use Rotational Drag=false)
   &SolveForcesAndVelocity(Clamp Velocity=false, Limit Acceleration=false)
   &ScaleSpriteSize(
    Uniform Curve Scale=4
    Uniform Curve Sprite Scale=DataInterface() {propertyValues {Curve {keys [
        {arriveTangent=9.0725, interpMode=RCIM_Cubic, leaveTangent=9.0725, tangentMode=RCTM_User, time=0, value=0}
        {arriveTangent=-0.0529, interpMode=RCIM_Cubic, leaveTangent=-0.0529, tangentMode=RCTM_User, time=0.3, value=0.9847}
        {arriveTangent=-1.377, interpMode=RCIM_Cubic, leaveTangent=-1.377, tangentMode=RCTM_User, time=0.8, value=0.65}
        {arriveTangent=-0.0405, interpMode=RCIM_Cubic, leaveTangent=-0.0405, tangentMode=RCTM_User, time=1, value=0}
       ]}}}
   )
   &ScaleMeshSize(Scale Factor=&Multiply_VectorByFloat(
     Vector=<1 1 1>
     Float=&FloatFromCurve(FloatCurve=DataInterface() {propertyValues {Curve {keys [
          {arriveTangent=0.6971, interpMode=RCIM_Cubic, leaveTangent=0.6971, tangentMode=RCTM_User, time=0, value=0}
          {arriveTangent=-0.0529, interpMode=RCIM_Cubic, leaveTangent=-0.0529, tangentMode=RCTM_User, time=0.2, value=1.2}
          {arriveTangent=-1.6278, interpMode=RCIM_Cubic, leaveTangent=-1.6278, tangentMode=RCTM_User, time=0.8, value=0.65}
          {arriveTangent=-0.0405, interpMode=RCIM_Cubic, leaveTangent=-0.0405, tangentMode=RCTM_User, time=1, value=0}
         ]}}})
    ))
  ]
  rendererClasses [&NiagaraMeshRendererProperties]
  renderers [{rendererClass=&NiagaraMeshRendererProperties, rendererIndex=0}]
 }
 Clumps {
  emitterUpdateScript [
   &EmitterState(Life Cycle Mode=::NewEnumerator1, Loop Delay=0.05, Loop Duration=0.9, UseLoopDelay=false)
   &SpawnRate(
    SpawnRate=&MakeCustomFloatFromBool(Bool=$User.bIsInWater, True Float=0, False Float=20)
    Use Spawn Probability=false
   )
  ]
  particleSpawnScript [
   /SetVariables(Particles.InitSpriteRotRate=&RandomRangeFloat(Minimum=-2, Maximum=2)) {isSetParameters=true, module=SetVariables_#4}
   &InitializeParticle(
    Color Maximum=<rgba 0.7578 0.7578 0.9304 1>
    Color Minimum=<rgba 0.687 0.687 0.8435 1>
    Color Mode=::NewEnumerator2
    Lifetime Max=0.5
    Lifetime Min=0.35
    Lifetime Mode=::NewEnumerator1
    Sprite Rotation Angle=&RandomRangeFloat(Minimum=0, Maximum=360)
    Sprite Rotation Mode=::NewEnumerator0
    Sprite Size Max=<15 30>
    Sprite Size Min=<10 25>
    Sprite Size Mode=::NewEnumerator2
    Sprite UV Mode=::NewEnumerator2
    UsePositionOffset=false
    Write Material Random=true
   )
   &FN_SetSpawnGroup(Number of Spawn Groups=2)
   &SkeletalMeshLocation(
    Bone / Socket Index=$Particles.SpawnGroup
    Bone Sampling Mode=::NewEnumerator4
    MaskSpawnGroup=false
    Orientation Sampling=::NewEnumerator1
    Sampled Position Offset=&MakeVector(
     X=&MaskFloatBySpawnGroup(
      Masked Float=&SelectFloatFromArray(
       Array Sampling Mode=::NewEnumerator0
       Direct Array Index=?
       Float Selection Array=DataInterface() {propertyValues {FloatData [-4,14]}}
       Random Seed=0
      )
      Passthrough Float=&SelectFloatFromArray(
       Array Sampling Mode=::NewEnumerator0
       Direct Array Index=?
       Float Selection Array=DataInterface() {propertyValues {FloatData [4,-14]}}
       Random Seed=0
      )
     )
     Y=&MakeFloatFromVector(Vector=&RandomRangeVector(Minimum=<0 -65 0>, Maximum=<0 -80 0>), Channel=::NewEnumerator1)
     Z=0
    )
    Skeletal Mesh=$System.Skeletal Mesh
   )
   &SkeletalMeshLocation(
    Bone / Socket Index=2
    Bone Sampling Mode=::NewEnumerator4
    MaskSpawnGroup=false
    Orientation Sampling=::NewEnumerator1
    Position Sampling=::NewEnumerator2
    Sampled Position Offset Coordinate Space:ENiagaraSkelMeshTransforms=::NewEnumerator2
    Skeletal Mesh=$System.Skeletal Mesh
   ) {module=SkeletalMeshLocation001}
   &AddVelocityInCone(
    Cone Angle=10
    Cone Axis=&RotateVectorByQuaternion(
     Vector To Rotate=&MaskVectorBySpawnGroup(Masked Vector=<0 -1 0>, Passthrough Vector=<0 -1 0>)
     Quaternion=$Particles.SkeletalMeshLocation.SampledRotation
    )
    Cone Axis Coordinate Space=::Simulation
    Module.Use Velocity Falloff On Cone Axis=true
    Velocity Distribution Along Cone Axis=0.75
    Velocity Falloff Away From Cone Axis=0.333
    Velocity Strength=&Multiply_Float(A=&RandomRangeFloat(Minimum=60, Maximum=80), B=$User.VelocityMult)
   )
   &AddVelocity(Velocity=<0 0 10>)
   /SetVariables(Particles.Lifetime=&MakeCustomFloatFromBool(
     Bool=&SetBoolByFloatComparison(
      A=50
      B=&DistanceBetweenPositions(
       A=&MakeVector(
        X=0
        Y=0
        Z=&MakeFloatFromVector(Vector=Linked() {linkedVariable {name=Particles.SkeletalMeshLocation001.SampledPosition, type=NiagaraPosition}}, Channel=::NewEnumerator2)
       )
       B=&MakeVector(
        X=0
        Y=0
        Z=&MakeFloatFromVector(Vector=Linked() {linkedVariable {name=Particles.SkeletalMeshLocation.SampledPosition, type=NiagaraPosition}}, Channel=::NewEnumerator2)
       )
      )
     )
     True Float=$Particles.Lifetime
     False Float=0
    )) {enabled=false, isSetParameters=true, module=SetVariables_#5}
  ]
  particleUpdateScript [
   &ParticleState()
   &GravityForce(Gravity=<0 0 -50>)
   &Drag(Drag=&RandomRangeFloat(Minimum=0.8, Maximum=1.2), Use Linear Drag=true, Use Rotational Drag=false)
   &BlendSpace(Rotation Contribution=0)
   &SolveForcesAndVelocity(Clamp Velocity=false, Limit Acceleration=false)
   &ScaleSpriteSize(Scale Factor=&Vector2DFromFloat(Value=&FloatFromCurve(FloatCurve=DataInterface() {propertyValues {Curve {keys [
          {arriveTangent=0.0905, interpMode=RCIM_Cubic, leaveTangent=0.0905, tangentMode=RCTM_User, time=0, value=0.5}
          {arriveTangent=5.1283, interpMode=RCIM_Cubic, leaveTangent=5.1283, tangentMode=RCTM_User, time=0.3, value=1.125}
          {arriveTangent=0.3828, interpMode=RCIM_Cubic, leaveTangent=0.3828, tangentMode=RCTM_User, time=0.8, value=3}
         ]}}})))
   /SetVariables(Particles.SmokeColor:LinearColor=&MultiplyLinearColor(
     Linear Color=&ColorFromCurve(ColorCurve=DataInterface() {propertyValues {
        AlphaCurve {keys [<at 0.8244 1 RCIM_Cubic>,<at 1 0 RCIM_Cubic>]}
        BlueCurve {keys [<at 0 2 RCIM_Cubic>]}
        GreenCurve {keys [<at 0 2 RCIM_Cubic>]}
        RedCurve {keys [<at 0 2 RCIM_Cubic>]}
       }})
     Scale Factor=&MakeVector4_FromLinearColor(LinearColor=<rgba 1 1 1 1>)
    )) {isSetParameters=true}
   /SetVariables(Particles.SmokeDyn:Vector4f=&MakeVector4(
     X=&FloatFromCurve(FloatCurve=DataInterface() {propertyValues {Curve {keys [<at 0 0>,<at 1 1>]}}})
     Y=&UniformAOrBFloat(A=0, Evaluation Type=::NewEnumerator0, Recalculate Random Each Loop=true, Override Seed=false)
     Z=0
     W=0
    )) {enabled=false, isSetParameters=true}
   &SubUVAnimation(
    End Frame Range Override=15
    "Sprite Renderer "=DataInterface() {propertyValues {SpriteRenderer=&self:Clumps_2.NiagaraSpriteRendererProperties_0}}
    UseEndFrame=false
    UseStartFrame=false
   ) {enabled=false}
   &ScaleColor(
    Scale RGB=&MakeVectorFromLinearColorRGB(Color=&ColorFromCurve(
      ColorCurve=DataInterface() {propertyValues {
        AlphaCurve {keys [<at 0 1 RCIM_Cubic>]}
        BlueCurve {keys [
          <at 0.0043 0.7>
          {arriveTangent=-0.2478, interpMode=RCIM_Cubic, leaveTangent=-0.2478, time=0.1789, value=0.7}
          <at 0.2812 0.6314>
          {arriveTangent=1.3318, interpMode=RCIM_Cubic, leaveTangent=1.3318, time=0.3527, value=1}
          {arriveTangent=-1.0217, interpMode=RCIM_Cubic, leaveTangent=-1.0217, time=0.558, value=1}
          {arriveTangent=-0.9848, interpMode=RCIM_Cubic, leaveTangent=-0.9848, time=0.8029, value=0.54}
          <at 0.8616 0.7011>
          <at 1 0.7>
         ]}
        GreenCurve {keys [
          <at 0.0043 0.35>
          {arriveTangent=1.1435, interpMode=RCIM_Cubic, leaveTangent=1.1435, time=0.1789, value=0.5785}
          <at 0.2812 0.6667>
          {arriveTangent=1.2043, interpMode=RCIM_Cubic, leaveTangent=1.2043, time=0.3527, value=1}
          {arriveTangent=-0.8536, interpMode=RCIM_Cubic, leaveTangent=-0.8536, time=0.558, value=1}
          {arriveTangent=-1.3917, interpMode=RCIM_Cubic, leaveTangent=-1.3917, time=0.8029, value=0.6157}
          <at 0.8616 0.5776>
          <at 1 0.35>
         ]}
        RedCurve {keys [
          <at 0.0043 0.28>
          {arriveTangent=2.246, interpMode=RCIM_Cubic, leaveTangent=2.246, time=0.1789, value=0.5542}
          <at 0.2812 0.902>
          {arriveTangent=0.3542, interpMode=RCIM_Cubic, leaveTangent=0.3542, time=0.3527, value=1}
          {arriveTangent=-0.2221, interpMode=RCIM_Cubic, leaveTangent=-0.2221, time=0.558, value=1}
          {arriveTangent=-1.476, interpMode=RCIM_Cubic, leaveTangent=-1.476, time=0.8029, value=0.9}
          <at 0.8616 0.552>
          <at 1 0.28>
         ]}
       }}
      CurveIndex=Linked() {linkedVariable {name=NPC.FortniteNPC.RemappedTimeOfDay_0-1, type=NiagaraFloat}}
     ))
    ScaleA=false
    ScaleRGB=true
   ) {module=ScaleColor001}
   &DynamicMaterialParameters(
    Index 0 Param 1=$Particles.NormalizedAge
    Module.Index0Float0Write=true
    Module.Index0Float1Write=true
    Module.Index0Float2Write=true
    Module.Index0Float3Write=true
   )
  ]
  rendererClasses [&NiagaraSpriteRendererProperties]
  renderers [
   {rendererClass=&NiagaraSpriteRendererProperties}
   {rendererClass=&NiagaraSpriteRendererProperties, rendererIndex=1}
  ]
 }
 Clumps001 {
  bEnabled=false
  emitterUpdateScript [
   &EmitterState()
   &SpawnRate(
    SpawnGroup=3
    SpawnRate=&FloatFromCurve(
     FloatCurve=DataInterface() {propertyValues {Curve {keys [<at 0.1 0 RCIM_Constant>,<at 0.2106 1 RCIM_Constant>,<at 0.2212 0 RCIM_Constant>,<at 0.2936 1 RCIM_Constant>,<at 0.3031 0 RCIM_Constant>,<at 0.3401 1 RCIM_Constant>,<at 0.3465 0 RCIM_Constant>,<at 0.381 1 RCIM_Constant>,<at 0.3884 0 RCIM_Constant>,<at 0.4331 1 RCIM_Constant>,<at 0.4399 0 RCIM_Constant>,<at 0.5345 1 RCIM_Constant>,<at 0.5584 1 RCIM_Constant>,<at 0.5636 0 RCIM_Constant>,<at 0.7232 1 RCIM_Constant>,<at 0.7369 0 RCIM_Constant>,<at 0.8612 1 RCIM_Constant>,<at 0.888 1 RCIM_Constant>,<at 0.8943 0 RCIM_Constant>]}}}
     CurveIndex=$Emitter.NormalizedLoopAge
     Scale Curve=20
    )
    Use Spawn Probability=false
   )
   &SpawnRate(
    SpawnGroup=4
    SpawnRate=&FloatFromCurve(
     FloatCurve=DataInterface() {propertyValues {Curve {keys [<at 0.1 0 RCIM_Constant>,<at 0.2106 1 RCIM_Constant>,<at 0.2212 0 RCIM_Constant>,<at 0.2936 1 RCIM_Constant>,<at 0.3031 0 RCIM_Constant>,<at 0.3401 1 RCIM_Constant>,<at 0.3465 0 RCIM_Constant>,<at 0.381 1 RCIM_Constant>,<at 0.3884 0 RCIM_Constant>,<at 0.4331 1 RCIM_Constant>,<at 0.4399 0 RCIM_Constant>,<at 0.5345 1 RCIM_Constant>,<at 0.5584 1 RCIM_Constant>,<at 0.5636 0 RCIM_Constant>,<at 0.7232 1 RCIM_Constant>,<at 0.7369 0 RCIM_Constant>,<at 0.8612 1 RCIM_Constant>,<at 0.888 1 RCIM_Constant>,<at 0.8943 0 RCIM_Constant>]}}}
     CurveIndex=$Emitter.NormalizedLoopAge
     Scale Curve=20
    )
    Use Spawn Probability=false
   ) {module=SpawnRate001}
  ]
  particleSpawnScript [
   /SetVariables(Particles.InitSpriteRotRate=&RandomRangeFloat(Minimum=-2, Maximum=2)) {isSetParameters=true, module=SetVariables_#4}
   &InitializeParticle(
    Lifetime Max=0.3
    Lifetime Min=0.25
    Lifetime Mode=::NewEnumerator1
    Sprite Rotation Angle=&RandomRangeFloat(Minimum=0, Maximum=360)
    Sprite Rotation Mode=::NewEnumerator0
    Sprite Size Mode=::NewEnumerator3
    Uniform Sprite Size=60
    UsePositionOffset=false
    Write Material Random=true
   )
   &FN_SetSpawnGroup(Number of Spawn Groups=2) {enabled=false}
   &SkeletalMeshLocation(
    Bone / Socket Index=&SpawnGroup()
    Bone Sampling Mode=::NewEnumerator4
    MaskSpawnGroup=false
    Skeletal Mesh=$System.Skeletal Mesh
   )
   &SphereLocation(
    MaskSpawnGroup=false
    Override Seed=false
    Sphere Coordinate Space=::Simulation
    Sphere Origin:NiagaraPosition=$Particles.SkeletalMeshLocation.SampledPosition
    Sphere Radius=8
    Surface Only=true
    Surface Only Band Thickness=1
    UseNonUniformScale=false
    UseOffset=false
    UseSphereOrientation=false
   )
   &AddVelocityInCone(
    Cone Angle=10
    Cone Axis=&RotateVectorByQuaternion(
     Vector To Rotate=&MaskVectorBySpawnGroup(Masked Vector=<0 -1 0>, Passthrough Vector=<0 -1 0>)
     Quaternion=$Particles.SkeletalMeshLocation.SampledRotation
    )
    Cone Axis Coordinate Space=::Simulation
    Module.Use Velocity Falloff On Cone Axis=true
    Velocity Distribution Along Cone Axis=0.75
    Velocity Falloff Away From Cone Axis=0.333
    Velocity Strength=&Multiply_Float(A=&RandomRangeFloat(Minimum=150, Maximum=180), B=$User.VelocityMult)
   ) {enabled=false}
   &AddVelocity(Velocity=&Multiply_VectorByFloat(Vector=$Output.SphereLocation.SphereNormal, Float=&RandomRangeFloat(Minimum=200, Maximum=300))) {module=AddVelocity001}
   &AddVelocity(Velocity=<0 0 30>)
   /SetVariables(Particles.Lifetime=&MakeCustomFloatFromBool(
     Bool=&SetBoolByFloatComparison(
      A=50
      B=&DistanceBetweenPositions(
       A=&MakeVector(
        X=0
        Y=0
        Z=&MakeFloatFromVector(Vector=Linked() {linkedVariable {name=Particles.SkeletalMeshLocation001.SampledPosition, type=NiagaraPosition}}, Channel=::NewEnumerator2)
       )
       B=&MakeVector(
        X=0
        Y=0
        Z=&MakeFloatFromVector(Vector=Linked() {linkedVariable {name=Particles.SkeletalMeshLocation.SampledPosition, type=NiagaraPosition}}, Channel=::NewEnumerator2)
       )
      )
     )
     True Float=$Particles.Lifetime
     False Float=0
    )) {enabled=false, isSetParameters=true, module=SetVariables_#5}
  ]
  particleUpdateScript [
   &ParticleState()
   &Drag(Drag=&RandomRangeFloat(Minimum=0.8, Maximum=1.2), Use Linear Drag=true, Use Rotational Drag=false)
   &SolveForcesAndVelocity(Clamp Velocity=false, Limit Acceleration=false)
   &SpriteRotationRate(Rotation Rate=&Multiply_Float(
     A=$Particles.InitSpriteRotRate
     B=&FloatFromCurve(
      FloatCurve=DataInterface() {propertyValues {Curve {keys [
          {arriveTangent=-4.0107, interpMode=RCIM_Cubic, leaveTangent=-4.0107, tangentMode=RCTM_User, time=-0.0046, value=1.0113}
          {arriveTangent=-0.2496, interpMode=RCIM_Cubic, leaveTangent=-0.2496, tangentMode=RCTM_User, time=0.5, value=0.16}
         ]}}}
      CurveIndex=$Emitter.Age
      Scale Curve=150
     )
    ))
   &Color(Color=&ColorFromCurve(ColorCurve=$User.SmokeColor)) {enabled=false}
   &ScaleColor(
    Color Value To Scale=$Particles.Color
    Scale Alpha=&FloatFromCurve(FloatCurve=DataInterface() {propertyValues {Curve {keys [{interpMode=RCIM_Cubic, tangentMode=RCTM_User, time=0.25, value=1}, {interpMode=RCIM_Cubic, tangentMode=RCTM_User, time=1, value=0}]}}})
    ScaleA=true
    ScaleRGB=true
   )
   &ScaleSpriteSize(Scale Factor=&Vector2DFromFloat(Value=&FloatFromCurve(FloatCurve=DataInterface() {propertyValues {Curve {keys [
          {arriveTangent=0.0905, interpMode=RCIM_Cubic, leaveTangent=0.0905, tangentMode=RCTM_User, time=0, value=0.5}
          {arriveTangent=5.1283, interpMode=RCIM_Cubic, leaveTangent=5.1283, tangentMode=RCTM_User, time=0.3, value=1.125}
          {arriveTangent=0.3828, interpMode=RCIM_Cubic, leaveTangent=0.3828, tangentMode=RCTM_User, time=0.8, value=2}
         ]}}})))
  ]
  rendererClasses [&NiagaraSpriteRendererProperties]
  renderers [{rendererClass=&NiagaraSpriteRendererProperties}]
 }
 DustSpikes {
  bEnabled=false
  emitterUpdateScript [
   &EmitterState()
   &SpawnRate(
    SpawnRate=&MakeCustomFloatFromBool(Bool=$User.bIsInWater, True Float=0, False Float=30)
    Use Spawn Probability=false
   )
  ]
  particleSpawnScript [
   &InitializeParticle(
    Lifetime Max=0.35
    Lifetime Min=0.2
    Lifetime Mode=::NewEnumerator1
    Mass Max=2
    Mass Mode=::NewEnumerator3
    Mesh Scale Max=<0.65 0 0>
    Mesh Scale Min=<0.25 0 0>
    Mesh Scale Mode=::NewEnumerator2
    UsePositionOffset=false
    Write Material Random=true
   )
   &FN_SetSpawnGroup(Number of Spawn Groups=2)
   &SkeletalMeshLocation(
    Bone / Socket Index=$Particles.SpawnGroup
    Bone Sampling Mode=::NewEnumerator4
    MaskSpawnGroup=false
    Orientation Sampling=::NewEnumerator1
    Sampled Position Offset=&MaskVectorBySpawnGroup(Masked Vector=<-8 0 0>, Passthrough Vector=<8 0 0>)
   )
   &AddVelocityInCone(
    Cone Angle=20
    Cone Axis=&RotateVectorByQuaternion(
     Vector To Rotate=&MaskVectorBySpawnGroup(Masked Vector=<0 -1 0>, Passthrough Vector=<0 13 0>)
     Quaternion=$Particles.SkeletalMeshLocation.SampledRotation
    )
    Cone Axis Coordinate Space=::Simulation
    Module.Use Velocity Falloff On Cone Axis=true
    Velocity Distribution Along Cone Axis=0.75
    Velocity Falloff Away From Cone Axis=0.333
    Velocity Strength=&RandomRangeFloat(Minimum=0.1, Maximum=0.2)
   )
  ]
  particleUpdateScript [
   &ParticleState()
   &Drag(Drag=&RandomRangeFloat(Minimum=0.8, Maximum=1.2), Use Linear Drag=true, Use Rotational Drag=false)
   &SolveForcesAndVelocity(Clamp Velocity=false, Limit Acceleration=false)
   &Color(Color=&ColorFromCurve(ColorCurve=$User.SnowColor))
   &ScaleColor(
    Color Value To Scale=$Particles.Color
    Scale Alpha=&FloatFromCurve(FloatCurve=DataInterface() {propertyValues {Curve {keys [
         {arriveTangent=0.0766, interpMode=RCIM_Cubic, leaveTangent=0.0766, tangentMode=RCTM_User, time=0, value=0}
         {arriveTangent=-2.5249, interpMode=RCIM_Cubic, leaveTangent=-2.5249, tangentMode=RCTM_User, time=0.1408, value=0.9847}
         {arriveTangent=-0.0405, interpMode=RCIM_Cubic, leaveTangent=-0.0405, tangentMode=RCTM_User, time=1, value=0}
        ]}}})
    Scale RGB=&RandomRangeVector(Minimum=$User.SnowColorMin, Maximum=$User.SnowColorMax)
    ScaleA=true
    ScaleRGB=true
   )
   &DynamicMaterialParameters(
    Index 0 Param 1=&RandomRangeFloat(Minimum=0, Maximum=1)
    Index 0 Param 3=&RandomRangeFloat(Minimum=0, Maximum=1)
    Index 0 Param 4=&Multiply_Float(
     A=&RandomRangeFloat(Minimum=0.25, Maximum=1)
     B=&FloatFromCurve(FloatCurve=DataInterface() {propertyValues {Curve {keys [{interpMode=RCIM_Cubic, tangentMode=RCTM_User, time=0, value=1}, {interpMode=RCIM_Cubic, tangentMode=RCTM_User, time=1, value=0.25}]}}})
    )
    Module.Param0WriteEnabled=false
    Module.Param1WriteEnabled=true
    Module.Param2WriteEnabled=false
    Module.Param3WriteEnabled=true
   )
   &ScaleMeshSize(Scale Factor=&Multiply_VectorByFloat(
     Vector=<1 1 1>
     Float=&FloatFromCurve(
      FloatCurve=DataInterface() {propertyValues {Curve {keys [
          {arriveTangent=2.6583, interpMode=RCIM_Cubic, leaveTangent=2.6583, tangentMode=RCTM_User, time=0, value=0.25}
          {arriveTangent=0.9995, interpMode=RCIM_Cubic, leaveTangent=0.9995, tangentMode=RCTM_User, time=0.2223, value=0.6825}
          {arriveTangent=-0.0405, interpMode=RCIM_Cubic, leaveTangent=-0.0405, tangentMode=RCTM_User, time=1, value=1}
         ]}}}
      Scale Curve=1.25
     )
    ))
  ]
  rendererClasses [&NiagaraMeshRendererProperties]
  renderers [{rendererClass=&NiagaraMeshRendererProperties, rendererIndex=0}]
 }
 DustSpikesSide {
  bEnabled=false
  emitterUpdateScript [
   &EmitterState()
   &SpawnRate(
    SpawnRate=&MakeCustomFloatFromBool(Bool=$User.bIsInWater, True Float=0, False Float=30)
    Use Spawn Probability=false
   )
  ]
  particleSpawnScript [
   &InitializeParticle(
    Lifetime Max=0.35
    Lifetime Min=0.2
    Lifetime Mode=::NewEnumerator1
    Mass Max=2
    Mass Mode=::NewEnumerator3
    Mesh Scale Max=<0.4 0 0>
    Mesh Scale Min=<0.1 0 0>
    Mesh Scale Mode=::NewEnumerator2
    UsePositionOffset=false
    Write Material Random=true
   )
   &FN_SetSpawnGroup(Number of Spawn Groups=2)
   &SkeletalMeshLocation(
    Bone / Socket Index=$Particles.SpawnGroup
    Bone Sampling Mode=::NewEnumerator4
    MaskSpawnGroup=false
    Orientation Sampling=::NewEnumerator1
    Sampled Position Offset=&MaskVectorBySpawnGroup(Masked Vector=<-8 10 20>, Passthrough Vector=<8 -10 -20>)
   ) {module=SkeletalMeshLocation001}
   &AddVelocityInCone(
    Cone Angle=20
    Cone Axis=&RotateVectorByQuaternion(
     Vector To Rotate=&MaskVectorBySpawnGroup(Masked Vector=<0 -1 0>, Passthrough Vector=<0 13 0>)
     Quaternion=$Particles.SkeletalMeshLocation.SampledRotation
    )
    Cone Axis Coordinate Space=::Simulation
    Module.Use Velocity Falloff On Cone Axis=true
    Velocity Distribution Along Cone Axis=0.75
    Velocity Falloff Away From Cone Axis=0.333
    Velocity Strength=&RandomRangeFloat(Minimum=0.1, Maximum=0.2)
   ) {enabled=false}
   &AddVelocity(Velocity=&MaskVectorBySpawnGroup(Masked Vector=<-50 0 10>, Passthrough Vector=<-50 0 10>))
   &InitialMeshOrientation(
    Mesh Orientation Mode=::NewEnumerator2
    Orientation Coordinate Space=::Local
    Orientation Vector=&MaskVectorBySpawnGroup(Masked Vector=<-1 0 1>, Passthrough Vector=<-1 0 1>)
    Use Rotation Vector=false
   )
   &CameraOffset(Camera Offset Amount=5)
  ]
  particleUpdateScript [
   &ParticleState()
   &Drag(Drag=&RandomRangeFloat(Minimum=0.8, Maximum=1.2), Use Linear Drag=true, Use Rotational Drag=false)
   &SolveForcesAndVelocity(Clamp Velocity=false, Limit Acceleration=false)
   &Color(Color=&ColorFromCurve(ColorCurve=$User.SnowColor))
   &ScaleColor(
    Color Value To Scale=$Particles.Color
    Scale Alpha=&FloatFromCurve(
     FloatCurve=DataInterface() {propertyValues {Curve {keys [
         {arriveTangent=0.0766, interpMode=RCIM_Cubic, leaveTangent=0.0766, tangentMode=RCTM_User, time=0, value=0}
         {arriveTangent=-2.5249, interpMode=RCIM_Cubic, leaveTangent=-2.5249, tangentMode=RCTM_User, time=0.1408, value=0.9847}
         {arriveTangent=-0.0405, interpMode=RCIM_Cubic, leaveTangent=-0.0405, tangentMode=RCTM_User, time=1, value=0}
        ]}}}
     Scale Curve=1.5
    )
    Scale RGB=&RandomRangeVector(Minimum=$User.SnowColorMin, Maximum=$User.SnowColorMax)
    ScaleA=true
    ScaleRGB=true
   )
   &DynamicMaterialParameters(
    Index 0 Param 1=&RandomRangeFloat(Minimum=0, Maximum=1)
    Index 0 Param 3=&RandomRangeFloat(Minimum=0, Maximum=1)
    Index 0 Param 4=&Multiply_Float(
     A=&RandomRangeFloat(Minimum=0.3, Maximum=1)
     B=&FloatFromCurve(FloatCurve=DataInterface() {propertyValues {Curve {keys [{interpMode=RCIM_Cubic, tangentMode=RCTM_User, time=0, value=1}, {interpMode=RCIM_Cubic, tangentMode=RCTM_User, time=1, value=0.25}]}}})
    )
    Module.Param0WriteEnabled=false
    Module.Param1WriteEnabled=true
    Module.Param2WriteEnabled=false
    Module.Param3WriteEnabled=true
   )
   &ScaleMeshSize(Scale Factor=&Multiply_VectorByFloat(
     Vector=<1 1 1>
     Float=&FloatFromCurve(
      FloatCurve=DataInterface() {propertyValues {Curve {keys [
          {arriveTangent=2.6583, interpMode=RCIM_Cubic, leaveTangent=2.6583, tangentMode=RCTM_User, time=0, value=0.25}
          {arriveTangent=0.9995, interpMode=RCIM_Cubic, leaveTangent=0.9995, tangentMode=RCTM_User, time=0.2223, value=0.6825}
          {arriveTangent=-0.0405, interpMode=RCIM_Cubic, leaveTangent=-0.0405, tangentMode=RCTM_User, time=1, value=1}
         ]}}}
      Scale Curve=1.25
     )
    ))
  ]
  rendererClasses [&NiagaraMeshRendererProperties]
  renderers [{rendererClass=&NiagaraMeshRendererProperties, rendererIndex=0}]
 }
 Flakes {
  bEnabled=false
  simTarget=GPUComputeSim
  emitterUpdateScript [
   &EmitterState()
   &SpawnRate(
    SpawnRate=&MakeCustomFloatFromBool(Bool=$User.bIsInWater, True Float=0, False Float=15)
    Use Spawn Probability=false
   )
  ]
  particleSpawnScript [
   /SetVariables(Particles.Intensity:NiagaraFloat=&FloatFromCurve(
     FloatCurve=DataInterface() {propertyValues {Curve {keys [<at 0 0>,<at 0.8 0>,<at 1 1>]}}}
     CurveIndex=&RandomRangeFloat(Minimum=0, Maximum=1)
    )) {isSetParameters=true, module=SetVariables_#2}
   &InitializeParticle(
    Lifetime=&Multiply_Float(
     A=&RandomRangeFloat(Minimum=0.15, Maximum=0.25)
     B=&Lerp_Float(A=0.8, B=1.4, Alpha=$Particles.Intensity)
    )
    Mass Max=2
    Mass Mode=::NewEnumerator3
    Mesh Scale Max=<0.65 0 0>
    Mesh Scale Min=<0.25 0 0>
    Mesh Scale Mode=::NewEnumerator2
    Sprite Rotation Mode=::NewEnumerator2
    Sprite Size Mode=::NewEnumerator3
    Uniform Sprite Size=&Multiply_Float(
     A=&RandomRangeFloat(Minimum=1.2, Maximum=1.4)
     B=&Lerp_Float(A=1, B=2, Alpha=$Particles.Intensity)
    )
    UsePositionOffset=false
    Write Material Random=true
   )
   &FN_SetSpawnGroup(Number of Spawn Groups=2)
   &SkeletalMeshLocation(
    Bone / Socket Index=$Particles.SpawnGroup
    Bone Sampling Mode=::NewEnumerator4
    MaskSpawnGroup=false
    Orientation Sampling=::NewEnumerator1
    Sampled Position Offset=&MaskVectorBySpawnGroup(Masked Vector=<0 -10 0>, Passthrough Vector=<0 10 0>)
   )
   &SphereLocation(
    MaskSpawnGroup=false
    Override Seed=false
    Sphere Coordinate Space=::Simulation
    Sphere Radius=5
    Surface Only=false
    UseNonUniformScale=false
    UseOffset=false
    UseSphereOrientation=false
   )
   &AddVelocityInCone(
    Cone Angle=40
    Cone Axis=&RotateVectorByQuaternion(
     Vector To Rotate=&MaskVectorBySpawnGroup(Masked Vector=<0 -1 0>, Passthrough Vector=<0 1 0>)
     Quaternion=$Particles.SkeletalMeshLocation.SampledRotation
    )
    Cone Axis Coordinate Space=::Simulation
    Module.Use Velocity Falloff On Cone Axis=true
    Velocity Distribution Along Cone Axis=0.75
    Velocity Falloff Away From Cone Axis=0.333
    Velocity Strength=&Multiply_Float(A=&RandomRangeFloat(Minimum=180, Maximum=200), B=$User.VelocityMult)
   )
   &AddVelocity(Velocity=&RandomRangeVector(Minimum=<0 0 40>, Maximum=<0 0 60>))
   &SubUVAnimation(End Frame=3, SubUV Animation Mode:ENiagara_SubUVLookupMode=::NewEnumerator2)
  ]
  particleUpdateScript [
   &ParticleState()
   &SpriteRotationRate(Rotation Rate=&RandomRangeFloat(Minimum=-2, Maximum=2))
   &GravityForce(Gravity=<0 0 -250>)
   &Drag(Drag=&RandomRangeFloat(Minimum=0.8, Maximum=1.2), Use Linear Drag=true, Use Rotational Drag=false)
   &Collision(
    Age Colliding Particles=false
    Kill Occluded Particles=false
    Kill Offscreen Particles=true
    Module.Control Rest State Via Surface Normal=false
    Module.EnableMaxCollisionCount=false
    Module.Randomize Collision Normal Vector=false
   )
   &SolveForcesAndVelocity(Clamp Velocity=false, Limit Acceleration=false)
   &Color(Color=&ColorFromCurve(ColorCurve=$User.SnowColor))
   &ScaleColor(
    Color Value To Scale=$Particles.Color
    Scale Alpha=&FloatFromCurve(
     FloatCurve=DataInterface() {propertyValues {Curve {keys [
         {arriveTangent=4.1785, interpMode=RCIM_Cubic, leaveTangent=4.1785, tangentMode=RCTM_User, time=0, value=0}
         {arriveTangent=0.9995, interpMode=RCIM_Cubic, leaveTangent=0.9995, tangentMode=RCTM_User, time=0.2223, value=0.6825}
         {arriveTangent=-0.0405, interpMode=RCIM_Cubic, leaveTangent=-0.0405, tangentMode=RCTM_User, time=1, value=1}
        ]}}}
     Scale Curve=&Lerp_Float(A=0.5, B=0.8, Alpha=$Particles.Initial.Intensity)
    )
    Scale RGB=&RandomRangeVector(Minimum=$User.SnowColorMin, Maximum=$User.SnowColorMax)
    ScaleA=true
    ScaleRGB=true
   ) {module=ScaleColor001}
   &ScaleColor(
    Color Value To Scale=$Particles.Color
    Scale Alpha=&FloatFromCurve(FloatCurve=DataInterface() {propertyValues {Curve {keys [
         {arriveTangent=9.0725, interpMode=RCIM_Cubic, leaveTangent=9.0725, tangentMode=RCTM_User, time=0, value=0}
         {arriveTangent=-2.5249, interpMode=RCIM_Cubic, leaveTangent=-2.5249, tangentMode=RCTM_User, time=0.1408, value=0.9847}
         {arriveTangent=-0.0405, interpMode=RCIM_Cubic, leaveTangent=-0.0405, tangentMode=RCTM_User, time=1, value=0}
        ]}}})
    Scale RGB=<0.8 0.8 0.8>
    ScaleA=false
    ScaleRGB=true
   )
   &DynamicMaterialParameters(
    Index 0 Param 1=&RandomRangeFloat(Minimum=0, Maximum=1)
    Index 0 Param 3=&RandomRangeFloat(Minimum=0, Maximum=1)
    Index 0 Param 4=&Multiply_Float(
     A=&RandomRangeFloat(Minimum=0.25, Maximum=1)
     B=&FloatFromCurve(FloatCurve=DataInterface() {propertyValues {Curve {keys [{interpMode=RCIM_Cubic, tangentMode=RCTM_User, time=0, value=1}, {interpMode=RCIM_Cubic, tangentMode=RCTM_User, time=1, value=0.25}]}}})
    )
    Module.Param0WriteEnabled=false
    Module.Param1WriteEnabled=true
    Module.Param2WriteEnabled=false
    Module.Param3WriteEnabled=true
   )
   &ScaleSpriteSize(
    Uniform Curve Scale=4
    Uniform Curve Sprite Scale=DataInterface() {propertyValues {Curve {keys [
        {arriveTangent=9.0725, interpMode=RCIM_Cubic, leaveTangent=9.0725, tangentMode=RCTM_User, time=0, value=0}
        {arriveTangent=-0.0529, interpMode=RCIM_Cubic, leaveTangent=-0.0529, tangentMode=RCTM_User, time=0.1408, value=0.9847}
        {arriveTangent=-1.146, interpMode=RCIM_Cubic, leaveTangent=-1.146, time=0.8, value=0.65}
        {arriveTangent=-0.0405, interpMode=RCIM_Cubic, leaveTangent=-0.0405, tangentMode=RCTM_User, time=1, value=0}
       ]}}}
   )
  ]
  rendererClasses [&NiagaraSpriteRendererProperties]
  renderers [{rendererClass=&NiagaraSpriteRendererProperties}]
 }
 FlakesSide {
  bEnabled=false
  simTarget=GPUComputeSim
  emitterUpdateScript [
   &EmitterState()
   &SpawnRate(
    SpawnRate=&MakeCustomFloatFromBool(Bool=$User.bIsInWater, True Float=0, False Float=60)
    Use Spawn Probability=false
   )
  ]
  particleSpawnScript [
   /SetVariables(Particles.Intensity:NiagaraFloat=&FloatFromCurve(
     FloatCurve=DataInterface() {propertyValues {Curve {keys [<at 0 0>,<at 0.8 0>,<at 1 1>]}}}
     CurveIndex=&RandomRangeFloat(Minimum=0, Maximum=1)
    )) {isSetParameters=true, module=SetVariables_#2}
   &InitializeParticle(
    Lifetime=&Multiply_Float(
     A=&RandomRangeFloat(Minimum=0.15, Maximum=0.25)
     B=&Lerp_Float(A=0.6, B=0.8, Alpha=$Particles.Intensity)
    )
    Mass Max=2
    Mass Mode=::NewEnumerator3
    Mesh Scale Max=<0.65 0 0>
    Mesh Scale Min=<0.25 0 0>
    Mesh Scale Mode=::NewEnumerator2
    Sprite Rotation Mode=::NewEnumerator2
    Sprite Size Mode=::NewEnumerator3
    Uniform Sprite Size=&Multiply_Float(
     A=&RandomRangeFloat(Minimum=0.8, Maximum=1.2)
     B=&Lerp_Float(A=0.6, B=0.8, Alpha=$Particles.Intensity)
    )
    UsePositionOffset=false
    Write Material Random=true
   )
   &FN_SetSpawnGroup(Number of Spawn Groups=2)
   &SkeletalMeshLocation(
    Bone / Socket Index=$Particles.SpawnGroup
    Bone Sampling Mode=::NewEnumerator4
    MaskSpawnGroup=false
    Orientation Sampling=::NewEnumerator1
    Sampled Position Offset=&MaskVectorBySpawnGroup(Masked Vector=<-8 10 20>, Passthrough Vector=<8 -10 -20>)
   )
   &SphereLocation(
    MaskSpawnGroup=false
    Override Seed=false
    Sphere Coordinate Space=::Simulation
    Sphere Radius=5
    Surface Only=false
    UseNonUniformScale=false
    UseOffset=false
    UseSphereOrientation=false
   )
   &AddVelocityInCone(
    Cone Angle=40
    Cone Axis=&RotateVectorByQuaternion(
     Vector To Rotate=&MaskVectorBySpawnGroup(Masked Vector=<0 -1 0>, Passthrough Vector=<0 1 0>)
     Quaternion=$Particles.SkeletalMeshLocation.SampledRotation
    )
    Cone Axis Coordinate Space=::Simulation
    Module.Use Velocity Falloff On Cone Axis=true
    Velocity Distribution Along Cone Axis=0.75
    Velocity Falloff Away From Cone Axis=0.333
    Velocity Strength=&Multiply_Float(A=&RandomRangeFloat(Minimum=180, Maximum=200), B=$User.VelocityMult)
   )
   &AddVelocity(Velocity=&RandomRangeVector(Minimum=<0 0 40>, Maximum=<0 0 60>))
   &SubUVAnimation(End Frame=3, SubUV Animation Mode:ENiagara_SubUVLookupMode=::NewEnumerator2)
  ]
  particleUpdateScript [
   &ParticleState()
   &SpriteRotationRate(Rotation Rate=&RandomRangeFloat(Minimum=-2, Maximum=2))
   &GravityForce(Gravity=<0 0 -250>)
   &Drag(Drag=&RandomRangeFloat(Minimum=0.8, Maximum=1.2), Use Linear Drag=true, Use Rotational Drag=false)
   &Collision(
    Advanced Aging Rate=1
    Age Colliding Particles=false
    Kill Occluded Particles=false
    Kill Offscreen Particles=true
    Module.Control Rest State Via Surface Normal=false
    Module.EnableMaxCollisionCount=false
    Module.Randomize Collision Normal Vector=false
   )
   &SolveForcesAndVelocity(Clamp Velocity=false, Limit Acceleration=false)
   &Color(Color=&ColorFromCurve(ColorCurve=$User.SnowColor))
   &ScaleColor(
    Color Value To Scale=$Particles.Color
    Scale Alpha=&FloatFromCurve(
     FloatCurve=DataInterface() {propertyValues {Curve {keys [
         {arriveTangent=4.1785, interpMode=RCIM_Cubic, leaveTangent=4.1785, tangentMode=RCTM_User, time=0, value=0}
         {arriveTangent=0.9995, interpMode=RCIM_Cubic, leaveTangent=0.9995, tangentMode=RCTM_User, time=0.2223, value=0.6825}
         {arriveTangent=-0.0405, interpMode=RCIM_Cubic, leaveTangent=-0.0405, tangentMode=RCTM_User, time=1, value=1}
        ]}}}
     Scale Curve=&Lerp_Float(A=0.5, B=0.8, Alpha=$Particles.Initial.Intensity)
    )
    Scale RGB=&RandomRangeVector(Minimum=$User.SnowColorMin, Maximum=$User.SnowColorMax)
    ScaleA=true
    ScaleRGB=true
   ) {module=ScaleColor001}
   &ScaleColor(
    Color Value To Scale=$Particles.Color
    Scale Alpha=&FloatFromCurve(FloatCurve=DataInterface() {propertyValues {Curve {keys [
         {arriveTangent=9.0725, interpMode=RCIM_Cubic, leaveTangent=9.0725, tangentMode=RCTM_User, time=0, value=0}
         {arriveTangent=-2.5249, interpMode=RCIM_Cubic, leaveTangent=-2.5249, tangentMode=RCTM_User, time=0.1408, value=0.9847}
         {arriveTangent=-0.0405, interpMode=RCIM_Cubic, leaveTangent=-0.0405, tangentMode=RCTM_User, time=1, value=0}
        ]}}})
    Scale RGB=<0.8 0.8 0.8>
    ScaleA=false
    ScaleRGB=true
   )
   &DynamicMaterialParameters(
    Index 0 Param 1=&RandomRangeFloat(Minimum=0, Maximum=1)
    Index 0 Param 3=&RandomRangeFloat(Minimum=0, Maximum=1)
    Index 0 Param 4=&Multiply_Float(
     A=&RandomRangeFloat(Minimum=0.25, Maximum=1)
     B=&FloatFromCurve(FloatCurve=DataInterface() {propertyValues {Curve {keys [{interpMode=RCIM_Cubic, tangentMode=RCTM_User, time=0, value=1}, {interpMode=RCIM_Cubic, tangentMode=RCTM_User, time=1, value=0.25}]}}})
    )
    Module.Param0WriteEnabled=false
    Module.Param1WriteEnabled=true
    Module.Param2WriteEnabled=false
    Module.Param3WriteEnabled=true
   )
   &ScaleSpriteSize(
    Uniform Curve Scale=4
    Uniform Curve Sprite Scale=DataInterface() {propertyValues {Curve {keys [
        {arriveTangent=9.0725, interpMode=RCIM_Cubic, leaveTangent=9.0725, tangentMode=RCTM_User, time=0, value=0}
        {arriveTangent=-0.0529, interpMode=RCIM_Cubic, leaveTangent=-0.0529, tangentMode=RCTM_User, time=0.3, value=0.9847}
        {arriveTangent=-1.377, interpMode=RCIM_Cubic, leaveTangent=-1.377, tangentMode=RCTM_User, time=0.8, value=0.65}
        {arriveTangent=-0.0405, interpMode=RCIM_Cubic, leaveTangent=-0.0405, tangentMode=RCTM_User, time=1, value=0}
       ]}}}
   )
  ]
  rendererClasses [&NiagaraSpriteRendererProperties]
  renderers [{rendererClass=&NiagaraSpriteRendererProperties}]
 }
 Goop {
  bEnabled=false
  emitterSpawnScript [/SetVariables() {isSetParameters=true}]
  emitterUpdateScript [
   &EmitterState()
   &SpawnRate(
    SpawnRate=&FloatFromCurve(
     FloatCurve=DataInterface() {propertyValues {Curve {keys [<at 0 1>,<at 0.1 1>,<at 0.2 1>,<at 1 1>]}}}
     CurveIndex=$Emitter.NormalizedLoopAge
     Scale Curve=150
    )
    Use Spawn Probability=false
   )
   &SpawnRate(
    SpawnGroup=1
    SpawnRate=&FloatFromCurve(
     FloatCurve=DataInterface() {propertyValues {Curve {keys [<at 0 1>,<at 0.1 1>,<at 0.2 1>,<at 1 1>]}}}
     CurveIndex=$Emitter.NormalizedLoopAge
     Scale Curve=150
    )
    Use Spawn Probability=false
   ) {module=SpawnRate001}
  ]
  particleSpawnScript [
   &InitializeParticle(Lifetime=0.24, Ribbon Width=8, Ribbon Width Mode=::NewEnumerator1, UsePositionOffset=false, Write Material Random=true)
   &FN_SetSpawnGroup(Pattern=::NewEnumerator0, Set Ribbon ID and Ribbon Link Order=true) {enabled=false}
   /SetVariables(Particles.RibbonID=/Niagara/DynamicInputs/Meshes/MakeRibbonIDFromSpawnGroup()) {isSetParameters=true}
   &CameraOffset(Camera Offset Amount=-15) {enabled=false}
   &SkeletalMeshLocation(
    MaskSpawnGroup=false
    Mesh Sampling Type=::NewEnumerator5
    Sampled Orientation Coordinate Space=::Local
    Sampled Position Coordinate Space=::Local
    Sampled Position Offset Coordinate Space:ENiagaraSkelMeshTransforms=::NewEnumerator2
    Sampled Velocity Coordinate Space:ENiagaraCoordinateSpace=::Local
    Skeletal Mesh=$System.Skeletal Mesh
    Socket Sampling Mode=::NewEnumerator5
   ) {module=SkeletalMeshLocation002}
   &SkeletalMeshLocation(
    Bone / Socket Index=2
    Bone Sampling Mode=::NewEnumerator4
    MaskSpawnGroup=false
    Orientation Sampling=::NewEnumerator1
    Position Sampling=::NewEnumerator2
    Sampled Position Offset Coordinate Space:ENiagaraSkelMeshTransforms=::NewEnumerator2
    Skeletal Mesh=$System.Skeletal Mesh
   ) {module=SkeletalMeshLocation001}
   &AddVelocityInCone(
    Cone Angle=10
    Cone Axis=&RotateVectorByQuaternion(
     Vector To Rotate=&MaskVectorBySpawnGroup(Masked Vector=<0 -1 0>, Passthrough Vector=<0 13 0>)
     Quaternion=$Particles.SkeletalMeshLocation.SampledRotation
    )
    Cone Axis Coordinate Space=::Simulation
    Module.Use Velocity Falloff On Cone Axis=true
    Velocity Distribution Along Cone Axis=0.75
    Velocity Falloff Away From Cone Axis=0.333
    Velocity Strength=&Multiply_Float(A=&RandomRangeFloat(Minimum=150, Maximum=180), B=$User.VelocityMult)
   ) {enabled=false}
   &AddVelocity(Velocity=<-300 0 100>)
   /SetVariables(Particles.Lifetime=&MakeCustomFloatFromBool(
     Bool=&SetBoolByFloatComparison(
      A=50
      B=&DistanceBetweenPositions(
       A=&MakeVector(
        X=0
        Y=0
        Z=&MakeFloatFromVector(Vector=Linked() {linkedVariable {name=Particles.SkeletalMeshLocation001.SampledPosition, type=NiagaraPosition}}, Channel=::NewEnumerator2)
       )
       B=&MakeVector(
        X=0
        Y=0
        Z=&MakeFloatFromVector(Vector=Linked() {linkedVariable {name=Particles.SkeletalMeshLocation002.SampledPosition, type=NiagaraPosition}}, Channel=::NewEnumerator2)
       )
      )
     )
     True Float=$Particles.Lifetime
     False Float=0
    )) {isSetParameters=true}
  ]
  particleUpdateScript [
   &ParticleState()
   &SolveForcesAndVelocity(Clamp Velocity=false, Limit Acceleration=false)
   &DynamicMaterialParameters(
    Index 0 Param 2=&FloatFromCurve(FloatCurve=DataInterface() {propertyValues {Curve {keys [
         <at 0 0.3>
         {arriveTangent=-2.5521, interpMode=RCIM_Cubic, leaveTangent=-2.5521, tangentMode=RCTM_User, time=0.1, value=0.75}
         {arriveTangent=-0.0873, interpMode=RCIM_Cubic, leaveTangent=0.0167, tangentMode=RCTM_Break, time=1, value=0}
        ]}}})
    Index 0 Param 3=1
    Index 0 Param 4=1
    Module.Param0WriteEnabled=true
    Module.Param1WriteEnabled=true
    Module.Param2WriteEnabled=true
    Module.Param3WriteEnabled=true
   )
   &ScaleRibbonWidth(Ribbon Width Scale=&FloatFromCurve(FloatCurve=DataInterface() {propertyValues {Curve {keys [
         {arriveTangent=0.3835, interpMode=RCIM_Cubic, leaveTangent=0.3835, tangentMode=RCTM_User, time=0, value=0.85}
         {arriveTangent=4.7926, interpMode=RCIM_Cubic, leaveTangent=1.8929, tangentMode=RCTM_Break, time=0.3, value=1.75}
         <at 0.5 2>
        ]}}})) {module=ScaleRibbonWidth001}
   &ScaleColor(
    Scale Alpha=&Multiply_Float(
     A=&FNTimeOfDayScalar(Night Scale=1.5, Morning Scale=2, Evening Scale=2)
     B=&FloatFromCurve(FloatCurve=DataInterface() {propertyValues {Curve {keys [<at 0 0.8>,<at 0.3 0.95>,<at 1 1>]}}})
    )
    ScaleA=true
    ScaleRGB=true
   )
  ]
  rendererClasses [&NiagaraRibbonRendererProperties]
  renderers [
   {rendererClass=&NiagaraRibbonRendererProperties}
   {rendererClass=&NiagaraRibbonRendererProperties, rendererIndex=1}
  ]
 }
 SmokeBack {
  bEnabled=false
  emitterUpdateScript [
   &EmitterState()
   &SpawnRate(
    SpawnRate=&MakeCustomFloatFromBool(Bool=$User.bIsInWater, True Float=0, False Float=20)
    Use Spawn Probability=false
   )
  ]
  particleSpawnScript [
   /SetVariables(Particles.InitSpriteRotRate=&RandomRangeFloat(Minimum=-2, Maximum=2)) {isSetParameters=true, module=SetVariables_#4}
   &InitializeParticle(
    Color Maximum=<rgba 0 0.5642 1 1>
    Color Minimum=<rgba 0 0.5642 1 1>
    Color Mode=::NewEnumerator2
    Lifetime Max=0.35
    Lifetime Min=0.3
    Lifetime Mode=::NewEnumerator1
    Sprite Rotation Angle=&RandomRangeFloat(Minimum=0, Maximum=360)
    Sprite Rotation Mode=::NewEnumerator0
    Sprite Size Mode=::NewEnumerator3
    Uniform Sprite Size=30
    UsePositionOffset=false
    Write Material Random=true
   )
   &FN_SetSpawnGroup(Number of Spawn Groups=2)
   &SkeletalMeshLocation(
    Bone / Socket Index=$Particles.SpawnGroup
    Bone Sampling Mode=::NewEnumerator4
    MaskSpawnGroup=false
    Orientation Sampling=::NewEnumerator1
    Sampled Position Offset=&MaskVectorBySpawnGroup(Masked Vector=<-8 -5 -5>, Passthrough Vector=<8 5 5>)
   )
   &AddVelocityInCone(
    Cone Angle=20
    Cone Axis=&RotateVectorByQuaternion(
     Vector To Rotate=&MaskVectorBySpawnGroup(Masked Vector=<0 -1 0>, Passthrough Vector=<0 13 0>)
     Quaternion=$Particles.SkeletalMeshLocation.SampledRotation
    )
    Cone Axis Coordinate Space=::Simulation
    Module.Use Velocity Falloff On Cone Axis=true
    Velocity Distribution Along Cone Axis=0.75
    Velocity Falloff Away From Cone Axis=0.333
    Velocity Strength=&Multiply_Float(A=&RandomRangeFloat(Minimum=150, Maximum=180), B=$User.VelocityMult)
   )
   &AddVelocity(Velocity=<0 0 10>)
  ]
  particleUpdateScript [
   &ParticleState()
   &Drag(Drag=&RandomRangeFloat(Minimum=0.8, Maximum=1.2), Use Linear Drag=true, Use Rotational Drag=false)
   &SolveForcesAndVelocity(Clamp Velocity=false, Limit Acceleration=false)
   &SpriteRotationRate(Rotation Rate=&Multiply_Float(
     A=$Particles.InitSpriteRotRate
     B=&FloatFromCurve(
      FloatCurve=DataInterface() {propertyValues {Curve {keys [
          {arriveTangent=-4.0107, interpMode=RCIM_Cubic, leaveTangent=-4.0107, tangentMode=RCTM_User, time=-0.0046, value=1.0113}
          {arriveTangent=-0.2496, interpMode=RCIM_Cubic, leaveTangent=-0.2496, tangentMode=RCTM_User, time=0.5, value=0.16}
         ]}}}
      CurveIndex=$Emitter.Age
      Scale Curve=150
     )
    ))
   &Color(Color=&ColorFromCurve(ColorCurve=$User.SmokeColor))
   &ScaleColor(
    Color Value To Scale=$Particles.Color
    Scale Alpha=&FloatFromCurve(
     FloatCurve=DataInterface() {propertyValues {Curve {keys [
         {arriveTangent=0.4811, interpMode=RCIM_Cubic, leaveTangent=0.4811, tangentMode=RCTM_User, time=0, value=0}
         {arriveTangent=-2.5249, interpMode=RCIM_Cubic, leaveTangent=-2.5249, tangentMode=RCTM_User, time=0.1408, value=0.9847}
         {arriveTangent=-0.0405, interpMode=RCIM_Cubic, leaveTangent=-0.0405, tangentMode=RCTM_User, time=1, value=0}
        ]}}}
     Scale Curve=0.5
    )
    Scale RGB=&RandomRangeVector(Minimum=$User.SnowColorMin, Maximum=$User.SnowColorMax)
    ScaleA=true
    ScaleRGB=true
   )
   &ScaleSpriteSize(Scale Factor=&Vector2DFromFloat(Value=&FloatFromCurve(FloatCurve=DataInterface() {propertyValues {Curve {keys [
          {arriveTangent=0.0905, interpMode=RCIM_Cubic, leaveTangent=0.0905, tangentMode=RCTM_User, time=0, value=0.5}
          {arriveTangent=5.1283, interpMode=RCIM_Cubic, leaveTangent=5.1283, tangentMode=RCTM_User, time=0.3, value=1.125}
          {arriveTangent=0.3828, interpMode=RCIM_Cubic, leaveTangent=0.3828, tangentMode=RCTM_User, time=0.8, value=2}
         ]}}})))
   &DynamicMaterialParameters(
    Index 0 Param 1=&Add_Float(
     A=$Particles.RandomFloat
     B=&FloatFromCurve(FloatCurve=DataInterface() {propertyValues {Curve {keys [
          {arriveTangent=2.7533, interpMode=RCIM_Cubic, leaveTangent=2.847, tangentMode=RCTM_Break, time=0, value=0}
          {arriveTangent=1.3581, interpMode=RCIM_Cubic, leaveTangent=1.3581, tangentMode=RCTM_User, time=0.2558, value=0.5125}
          {arriveTangent=0.0529, interpMode=RCIM_Cubic, leaveTangent=0.2021, tangentMode=RCTM_User, time=1, value=1}
         ]}}})
    )
    Index 0 Param 3=&FloatFromCurve(
     FloatCurve=DataInterface() {propertyValues {Curve {keys [{arriveTangent=-53.9199, time=0, value=1}, {leaveTangent=0.2723, time=0.25, value=1}]}}}
     Scale Curve=0.1
    )
    Index 0 Param 4=&FloatFromCurve(FloatCurve=DataInterface() {propertyValues {Curve {keys [<at 0 1>,<at 1 0>]}}})
    Index 1 Param 1=&FloatFromCurve(FloatCurve=DataInterface() {propertyValues {Curve {keys [<at 0 0>, {arriveTangent=4.2257, interpMode=RCIM_Cubic, leaveTangent=4.2257, tangentMode=RCTM_User, time=0.25, value=0}]}}})
    Index 1 Param 2=&FloatFromCurve(FloatCurve=DataInterface() {propertyValues {Curve {keys [<at 0 0>]}}})
    Index 1 Param 3=&FloatFromCurve(FloatCurve=DataInterface() {propertyValues {Curve {keys [{interpMode=RCIM_Cubic, leaveTangent=6.3917, tangentMode=RCTM_Break, time=0, value=0}]}}})
    Index 1 Param 4=&FloatFromCurve(FloatCurve=DataInterface() {propertyValues {Curve {keys [{arriveTangent=-0.136, leaveTangent=-0.136, time=0, value=1}]}}})
    Index 2 Param 1=&FloatFromCurve(FloatCurve=DataInterface() {propertyValues {Curve {keys [
         {arriveTangent=2.646, interpMode=RCIM_Cubic, leaveTangent=2.646, tangentMode=RCTM_User, time=0, value=0}
         {leaveTangent=-0.0719, time=1, value=1}
        ]}}})
    Index 2 Param 2=&FloatFromCurve(
     FloatCurve=DataInterface() {propertyValues {Curve {keys [
         {arriveTangent=1.9056, time=0, value=0.2}
         {arriveTangent=0.0259, interpMode=RCIM_Cubic, leaveTangent=0.0259, tangentMode=RCTM_User, time=1, value=1.4}
        ]}}}
     Scale Curve=&RandomRangeFloat(Minimum=0.5, Maximum=1)
    )
    Index 2 Param 3=&FloatFromCurve(FloatCurve=DataInterface() {propertyValues {Curve {keys [
         {interpMode=RCIM_Cubic, leaveTangent=2.3064, tangentMode=RCTM_Break, time=0, value=1}
         {interpMode=RCIM_Cubic, leaveTangent=11.1233, tangentMode=RCTM_Break, time=1, value=2}
        ]}}})
    Index 2 Param 4=&FloatFromCurve(
     FloatCurve=DataInterface() {propertyValues {Curve {keys [
         {arriveTangent=-0.6867, interpMode=RCIM_Cubic, leaveTangent=-0.6867, tangentMode=RCTM_User, time=0, value=1}
         {interpMode=RCIM_Cubic, leaveTangent=-1.2956, tangentMode=RCTM_Break, time=1, value=0.75}
        ]}}}
     Scale Curve=1.6
    )
    Module.Param0WriteEnabled=true
    Module.Param0WriteEnabled1=true
    Module.Param0WriteEnabled2=true
    Module.Param1WriteEnabled=false
    Module.Param1WriteEnabled1=true
    Module.Param1WriteEnabled2=true
    Module.Param2WriteEnabled=true
    Module.Param2WriteEnabled1=true
    Module.Param2WriteEnabled2=true
    Module.Param3WriteEnabled=true
    Module.Param3WriteEnabled1=true
    Module.Param3WriteEnabled2=true
    Write Parameter Index 1=true
    Write Parameter Index 2=true
   )
  ]
  rendererClasses [&NiagaraSpriteRendererProperties]
  renderers [
   {rendererClass=&NiagaraSpriteRendererProperties}
   {rendererClass=&NiagaraSpriteRendererProperties, rendererIndex=1}
  ]
 }
 SmokeSides {
  bEnabled=false
  emitterUpdateScript [
   &EmitterState()
   &SpawnRate(
    SpawnRate=&MakeCustomFloatFromBool(Bool=$User.bIsInWater, True Float=0, False Float=20)
    Use Spawn Probability=false
   )
  ]
  particleSpawnScript [
   /SetVariables(Particles.InitSpriteRotRate=&RandomRangeFloat(Minimum=-2, Maximum=2)) {isSetParameters=true, module=SetVariables_#4}
   &InitializeParticle(
    Color Maximum=<rgba 0 0.5642 1 1>
    Color Minimum=<rgba 0 0.5642 1 1>
    Color Mode=::NewEnumerator2
    Lifetime Max=0.45
    Lifetime Min=0.35
    Lifetime Mode=::NewEnumerator1
    Sprite Rotation Angle=&RandomRangeFloat(Minimum=0, Maximum=360)
    Sprite Rotation Mode=::NewEnumerator0
    Sprite Size Mode=::NewEnumerator3
    Uniform Sprite Size=30
    UsePositionOffset=false
    Write Material Random=true
   )
   &FN_SetSpawnGroup(Number of Spawn Groups=2)
   &SkeletalMeshLocation(
    Bone / Socket Index=$Particles.SpawnGroup
    Bone Sampling Mode=::NewEnumerator4
    MaskSpawnGroup=false
    Orientation Sampling=::NewEnumerator1
    Sampled Position Offset=&MaskVectorBySpawnGroup(Masked Vector=<-8 10 20>, Passthrough Vector=<8 -10 -20>)
   )
   &AddVelocityInCone(
    Cone Angle=20
    Cone Axis=&RotateVectorByQuaternion(
     Vector To Rotate=&MaskVectorBySpawnGroup(Masked Vector=<0 -1 0>, Passthrough Vector=<0 13 0>)
     Quaternion=$Particles.SkeletalMeshLocation.SampledRotation
    )
    Cone Axis Coordinate Space=::Simulation
    Module.Use Velocity Falloff On Cone Axis=true
    Velocity Distribution Along Cone Axis=0.75
    Velocity Falloff Away From Cone Axis=0.333
    Velocity Strength=&Multiply_Float(A=&RandomRangeFloat(Minimum=150, Maximum=180), B=$User.VelocityMult)
   )
   &AddVelocity(Velocity=<0 0 5>)
  ]
  particleUpdateScript [
   &ParticleState()
   &Drag(Drag=&RandomRangeFloat(Minimum=0.8, Maximum=1.2), Use Linear Drag=true, Use Rotational Drag=false)
   &SolveForcesAndVelocity(Clamp Velocity=false, Limit Acceleration=false)
   &SpriteRotationRate(Rotation Rate=&Multiply_Float(
     A=$Particles.InitSpriteRotRate
     B=&FloatFromCurve(
      FloatCurve=DataInterface() {propertyValues {Curve {keys [
          {arriveTangent=-4.0107, interpMode=RCIM_Cubic, leaveTangent=-4.0107, tangentMode=RCTM_User, time=-0.0046, value=1.0113}
          {arriveTangent=-0.2496, interpMode=RCIM_Cubic, leaveTangent=-0.2496, tangentMode=RCTM_User, time=0.5, value=0.16}
         ]}}}
      CurveIndex=$Emitter.Age
      Scale Curve=150
     )
    ))
   &Color(Color=&ColorFromCurve(ColorCurve=$User.SmokeColor))
   &ScaleColor(
    Color Value To Scale=$Particles.Color
    Scale Alpha=&FloatFromCurve(
     FloatCurve=DataInterface() {propertyValues {Curve {keys [
         {arriveTangent=9.0725, interpMode=RCIM_Cubic, leaveTangent=9.0725, tangentMode=RCTM_User, time=0, value=0}
         {arriveTangent=-2.5249, interpMode=RCIM_Cubic, leaveTangent=-2.5249, tangentMode=RCTM_User, time=0.1408, value=0.9847}
         {arriveTangent=-0.0405, interpMode=RCIM_Cubic, leaveTangent=-0.0405, tangentMode=RCTM_User, time=1, value=0}
        ]}}}
     Scale Curve=1.5
    )
    Scale RGB=&RandomRangeVector(Minimum=$User.SnowColorMin, Maximum=$User.SnowColorMax)
    ScaleA=true
    ScaleRGB=true
   )
   &ScaleSpriteSize(Scale Factor=&Vector2DFromFloat(Value=&FloatFromCurve(FloatCurve=DataInterface() {propertyValues {Curve {keys [
          {arriveTangent=0.0905, interpMode=RCIM_Cubic, leaveTangent=0.0905, tangentMode=RCTM_User, time=0, value=0.5}
          {arriveTangent=5.1283, interpMode=RCIM_Cubic, leaveTangent=5.1283, tangentMode=RCTM_User, time=0.3, value=1.125}
          {arriveTangent=0.3828, interpMode=RCIM_Cubic, leaveTangent=0.3828, tangentMode=RCTM_User, time=0.8, value=2}
         ]}}})))
   &DynamicMaterialParameters(
    Index 0 Param 1=&Add_Float(
     A=$Particles.RandomFloat
     B=&FloatFromCurve(FloatCurve=DataInterface() {propertyValues {Curve {keys [
          {arriveTangent=2.7533, interpMode=RCIM_Cubic, leaveTangent=2.847, tangentMode=RCTM_Break, time=0, value=0}
          {arriveTangent=1.3581, interpMode=RCIM_Cubic, leaveTangent=1.3581, tangentMode=RCTM_User, time=0.2558, value=0.5125}
          {arriveTangent=0.0529, interpMode=RCIM_Cubic, leaveTangent=0.2021, tangentMode=RCTM_User, time=1, value=1}
         ]}}})
    )
    Index 0 Param 3=&FloatFromCurve(
     FloatCurve=DataInterface() {propertyValues {Curve {keys [{arriveTangent=-53.9199, time=0, value=1}, {leaveTangent=0.2723, time=0.25, value=1}]}}}
     Scale Curve=0.1
    )
    Index 0 Param 4=&FloatFromCurve(FloatCurve=DataInterface() {propertyValues {Curve {keys [<at 0 1>,<at 1 0>]}}})
    Index 1 Param 1=&FloatFromCurve(FloatCurve=DataInterface() {propertyValues {Curve {keys [<at 0 0>, {arriveTangent=4.2257, interpMode=RCIM_Cubic, leaveTangent=4.2257, tangentMode=RCTM_User, time=0.25, value=0}]}}})
    Index 1 Param 2=&FloatFromCurve(FloatCurve=DataInterface() {propertyValues {Curve {keys [<at 0 0>]}}})
    Index 1 Param 3=&FloatFromCurve(FloatCurve=DataInterface() {propertyValues {Curve {keys [{interpMode=RCIM_Cubic, leaveTangent=6.3917, tangentMode=RCTM_Break, time=0, value=0}]}}})
    Index 1 Param 4=&FloatFromCurve(FloatCurve=DataInterface() {propertyValues {Curve {keys [{arriveTangent=-0.136, leaveTangent=-0.136, time=0, value=1}]}}})
    Index 2 Param 1=&FloatFromCurve(FloatCurve=DataInterface() {propertyValues {Curve {keys [
         {arriveTangent=2.646, interpMode=RCIM_Cubic, leaveTangent=2.646, tangentMode=RCTM_User, time=0, value=0}
         {leaveTangent=-0.0719, time=1, value=1}
        ]}}})
    Index 2 Param 2=&FloatFromCurve(
     FloatCurve=DataInterface() {propertyValues {Curve {keys [
         {arriveTangent=1.9056, time=0, value=0.2}
         {arriveTangent=0.0259, interpMode=RCIM_Cubic, leaveTangent=0.0259, tangentMode=RCTM_User, time=1, value=1.4}
        ]}}}
     Scale Curve=&RandomRangeFloat(Minimum=0.5, Maximum=1)
    )
    Index 2 Param 3=&FloatFromCurve(FloatCurve=DataInterface() {propertyValues {Curve {keys [
         {interpMode=RCIM_Cubic, leaveTangent=2.3064, tangentMode=RCTM_Break, time=0, value=1}
         {interpMode=RCIM_Cubic, leaveTangent=11.1233, tangentMode=RCTM_Break, time=1, value=2}
        ]}}})
    Index 2 Param 4=&FloatFromCurve(
     FloatCurve=DataInterface() {propertyValues {Curve {keys [
         {arriveTangent=-0.6867, interpMode=RCIM_Cubic, leaveTangent=-0.6867, tangentMode=RCTM_User, time=0, value=1}
         {interpMode=RCIM_Cubic, leaveTangent=-1.2956, tangentMode=RCTM_Break, time=1, value=0.75}
        ]}}}
     Scale Curve=1.6
    )
    Module.Param0WriteEnabled=true
    Module.Param0WriteEnabled1=true
    Module.Param0WriteEnabled2=true
    Module.Param1WriteEnabled=false
    Module.Param1WriteEnabled1=true
    Module.Param1WriteEnabled2=true
    Module.Param2WriteEnabled=true
    Module.Param2WriteEnabled1=true
    Module.Param2WriteEnabled2=true
    Module.Param3WriteEnabled=true
    Module.Param3WriteEnabled1=true
    Module.Param3WriteEnabled2=true
    Write Parameter Index 1=true
    Write Parameter Index 2=true
   )
  ]
  rendererClasses [&NiagaraSpriteRendererProperties]
  renderers [
   {rendererClass=&NiagaraSpriteRendererProperties}
   {rendererClass=&NiagaraSpriteRendererProperties, rendererIndex=1}
  ]
 }
 Splash {
  bEnabled=false
  emitterUpdateScript [
   &EmitterState()
   &SpawnRate(
    SpawnRate=&MakeCustomFloatFromBool(Bool=$User.bIsInWater, True Float=0, False Float=10)
    Use Spawn Probability=false
   )
  ]
  particleSpawnScript [
   &InitializeParticle(
    Lifetime Max=0.35
    Lifetime Min=0.25
    Lifetime Mode=::NewEnumerator1
    Mass Max=2
    Mass Mode=::NewEnumerator3
    Mesh Scale Max=<0.35 0.35 0.01>
    Mesh Scale Min=<0.25 0.25 0.01>
    Mesh Scale Mode=::NewEnumerator2
    UsePositionOffset=false
    Write Material Random=true
   )
   &FN_SetSpawnGroup(Number of Spawn Groups=2)
   &SkeletalMeshLocation(
    Bone / Socket Index=$Particles.SpawnGroup
    Bone Sampling Mode=::NewEnumerator4
    MaskSpawnGroup=false
    Orientation Sampling=::NewEnumerator1
    Sampled Position Offset=&MaskVectorBySpawnGroup(Masked Vector=<0 -5 0>, Passthrough Vector=<0 -5 0>)
   ) {module=SkeletalMeshLocation001}
   &AddVelocity(Velocity=&MaskVectorBySpawnGroup(Masked Vector=<-50 0 10>, Passthrough Vector=<-50 0 10>)) {enabled=false}
   &InitialMeshOrientation(
    Mesh Axis To Orient=<0 0 0>
    Mesh Orientation Mode=::NewEnumerator2
    Orientation Coordinate Space=::Local
    Orientation Vector=&MaskVectorBySpawnGroup(Masked Vector=<-1 -0.2 -0.15>, Passthrough Vector=<-1 0.2 -0.15>)
    Rotation=&MaskVectorBySpawnGroup(
     Masked Vector=&RandomRangeVector(Minimum=<0 0 -0.05>, Maximum=<0 0 -0.1>)
     Passthrough Vector=&RandomRangeVector(Minimum=<0 0 0.05>, Maximum=<0 0 0.1>)
    )
    Use Rotation Vector=true
   )
   &CameraOffset(Camera Offset Amount=5) {enabled=false}
  ]
  particleUpdateScript [
   &ParticleState()
   &Drag(Drag=&RandomRangeFloat(Minimum=0.8, Maximum=1.2), Use Linear Drag=true, Use Rotational Drag=false)
   &SolveForcesAndVelocity(Clamp Velocity=false, Limit Acceleration=false)
   &Color()
   &ScaleColor(
    Color Value To Scale=$Particles.Color
    Scale Alpha=&FloatFromCurve(FloatCurve=DataInterface() {propertyValues {Curve {keys [
         {arriveTangent=0.0766, interpMode=RCIM_Cubic, leaveTangent=0.0766, tangentMode=RCTM_User, time=0, value=0}
         {arriveTangent=-2.5249, interpMode=RCIM_Cubic, leaveTangent=-2.5249, tangentMode=RCTM_User, time=0.1408, value=0.9847}
         {arriveTangent=-0.0405, interpMode=RCIM_Cubic, leaveTangent=-0.0405, tangentMode=RCTM_User, time=1, value=0}
        ]}}})
    Scale RGB=&RandomRangeVector(Minimum=$User.SnowColorMin, Maximum=$User.SnowColorMax)
    ScaleA=true
    ScaleRGB=true
   )
   &DynamicMaterialParameters(
    Index 0 Param 1=&RandomRangeFloat(Minimum=0, Maximum=1)
    Index 0 Param 3=&RandomRangeFloat(Minimum=0, Maximum=1)
    Index 0 Param 4=&Multiply_Float(
     A=&RandomRangeFloat(Minimum=0.25, Maximum=1)
     B=&FloatFromCurve(FloatCurve=DataInterface() {propertyValues {Curve {keys [{interpMode=RCIM_Cubic, tangentMode=RCTM_User, time=0, value=1}, {interpMode=RCIM_Cubic, tangentMode=RCTM_User, time=1, value=0.25}]}}})
    )
    Module.Param0WriteEnabled=false
    Module.Param1WriteEnabled=true
    Module.Param2WriteEnabled=false
    Module.Param3WriteEnabled=true
   ) {enabled=false}
   &ScaleMeshSize(Scale Factor=&Multiply_VectorByFloat(
     Vector=<1 1 1>
     Float=&FloatFromCurve(
      FloatCurve=DataInterface() {propertyValues {Curve {keys [
          {arriveTangent=2.6583, interpMode=RCIM_Cubic, leaveTangent=2.6583, tangentMode=RCTM_User, time=0, value=0.25}
          {arriveTangent=0.9995, interpMode=RCIM_Cubic, leaveTangent=0.9995, tangentMode=RCTM_User, time=0.2223, value=0.6825}
          {arriveTangent=-0.0405, interpMode=RCIM_Cubic, leaveTangent=-0.0405, tangentMode=RCTM_User, time=1, value=1}
         ]}}}
      Scale Curve=2
     )
    ))
   &DynamicMaterialParameters(
    Index 0 Param 1=&FloatFromCurve(FloatCurve=DataInterface() {propertyValues {Curve {keys [
         {arriveTangent=1.2629, interpMode=RCIM_Cubic, leaveTangent=1.2629, tangentMode=RCTM_User, time=-0.0045, value=-1.9799}
         {arriveTangent=4.4591, interpMode=RCIM_Cubic, leaveTangent=4.4591, tangentMode=RCTM_User, time=1, value=0}
        ]}}})
    Index 0 Param 2=&FloatFromCurve(FloatCurve=DataInterface() {propertyValues {Curve {keys [<at 0.0102 4>,<at 0.9749 0>]}}})
    Index 0 Param 3=&MakeFloatFromVector(Vector=$Particles.Scale, Channel=::NewEnumerator0)
    Index 0 Param 4=1
    Index 1 Param 1=&FloatFromCurve(FloatCurve=DataInterface() {propertyValues {Curve {keys [<at 0.7 0.15 RCIM_Cubic>, {arriveTangent=6.4641, interpMode=RCIM_Cubic, leaveTangent=6.4641, tangentMode=RCTM_User, time=1, value=1}]}}})
    Module.Param0WriteEnabled=true
    Module.Param0WriteEnabled1=true
    Module.Param1WriteEnabled=true
    Module.Param1WriteEnabled1=true
    Module.Param2WriteEnabled=true
    Module.Param2WriteEnabled1=true
    Module.Param3WriteEnabled=true
    Module.Param3WriteEnabled1=true
    Write Parameter Index 1=true
   ) {module=DynamicMaterialParameters001}
  ]
  rendererClasses [&NiagaraMeshRendererProperties]
  renderers [{rendererClass=&NiagaraMeshRendererProperties, rendererIndex=0}]
 }
 SplashTop {
  bEnabled=false
  emitterUpdateScript [
   &EmitterState()
   &SpawnRate(
    SpawnRate=&MakeCustomFloatFromBool(Bool=$User.bIsInWater, True Float=0, False Float=20)
    Use Spawn Probability=false
   )
  ]
  particleSpawnScript [
   &InitializeParticle(
    Lifetime Max=0.35
    Lifetime Min=0.25
    Lifetime Mode=::NewEnumerator1
    Mass Max=2
    Mass Mode=::NewEnumerator3
    Mesh Scale Max=<0.35 0.35 0.01>
    Mesh Scale Min=<0.25 0.25 0.01>
    Mesh Scale Mode=::NewEnumerator2
    UsePositionOffset=false
    Write Material Random=true
   )
   &FN_SetSpawnGroup(Number of Spawn Groups=2)
   &SkeletalMeshLocation(
    Bone / Socket Index=$Particles.SpawnGroup
    Bone Sampling Mode=::NewEnumerator4
    MaskSpawnGroup=false
    Orientation Sampling=::NewEnumerator1
    Sampled Position Offset=&MaskVectorBySpawnGroup(Masked Vector=<0 -5 0>, Passthrough Vector=<0 -5 0>)
   ) {module=SkeletalMeshLocation001}
   &AddVelocity(Velocity=&MaskVectorBySpawnGroup(Masked Vector=<-50 0 10>, Passthrough Vector=<-50 0 10>)) {enabled=false}
   &InitialMeshOrientation(
    Mesh Axis To Orient=<0 0 0>
    Mesh Orientation Mode=::NewEnumerator2
    Orientation Coordinate Space=::Local
    Orientation Vector=&MaskVectorBySpawnGroup(Masked Vector=<-1 -0.2 -0.15>, Passthrough Vector=<-1 0.2 -0.15>)
    Rotation=&MaskVectorBySpawnGroup(
     Masked Vector=&RandomRangeVector(Minimum=<0.15 0 -0.05>, Maximum=<0.15 0 -0.1>)
     Passthrough Vector=&RandomRangeVector(Minimum=<-0.15 0 0.05>, Maximum=<-0.15 0 0.1>)
    )
    Use Rotation Vector=true
   )
   &CameraOffset(Camera Offset Amount=5) {enabled=false}
  ]
  particleUpdateScript [
   &ParticleState()
   &Drag(Drag=&RandomRangeFloat(Minimum=0.8, Maximum=1.2), Use Linear Drag=true, Use Rotational Drag=false)
   &SolveForcesAndVelocity(Clamp Velocity=false, Limit Acceleration=false)
   &Color()
   &ScaleColor(
    Color Value To Scale=$Particles.Color
    Scale Alpha=&FloatFromCurve(FloatCurve=DataInterface() {propertyValues {Curve {keys [
         {arriveTangent=0.0766, interpMode=RCIM_Cubic, leaveTangent=0.0766, tangentMode=RCTM_User, time=0, value=0}
         {arriveTangent=-2.5249, interpMode=RCIM_Cubic, leaveTangent=-2.5249, tangentMode=RCTM_User, time=0.1408, value=0.9847}
         {arriveTangent=-0.0405, interpMode=RCIM_Cubic, leaveTangent=-0.0405, tangentMode=RCTM_User, time=1, value=0}
        ]}}})
    Scale RGB=&RandomRangeVector(Minimum=$User.SnowColorMin, Maximum=$User.SnowColorMax)
    ScaleA=true
    ScaleRGB=true
   )
   &DynamicMaterialParameters(
    Index 0 Param 1=&RandomRangeFloat(Minimum=0, Maximum=1)
    Index 0 Param 3=&RandomRangeFloat(Minimum=0, Maximum=1)
    Index 0 Param 4=&Multiply_Float(
     A=&RandomRangeFloat(Minimum=0.25, Maximum=1)
     B=&FloatFromCurve(FloatCurve=DataInterface() {propertyValues {Curve {keys [{interpMode=RCIM_Cubic, tangentMode=RCTM_User, time=0, value=1}, {interpMode=RCIM_Cubic, tangentMode=RCTM_User, time=1, value=0.25}]}}})
    )
    Module.Param0WriteEnabled=false
    Module.Param1WriteEnabled=true
    Module.Param2WriteEnabled=false
    Module.Param3WriteEnabled=true
   ) {enabled=false}
   &ScaleMeshSize(Scale Factor=&Multiply_VectorByFloat(
     Vector=<1 1 1>
     Float=&FloatFromCurve(
      FloatCurve=DataInterface() {propertyValues {Curve {keys [
          {arriveTangent=2.6583, interpMode=RCIM_Cubic, leaveTangent=2.6583, tangentMode=RCTM_User, time=0, value=0.25}
          {arriveTangent=0.9995, interpMode=RCIM_Cubic, leaveTangent=0.9995, tangentMode=RCTM_User, time=0.2223, value=0.6825}
          {arriveTangent=-0.0405, interpMode=RCIM_Cubic, leaveTangent=-0.0405, tangentMode=RCTM_User, time=1, value=1}
         ]}}}
      Scale Curve=1.25
     )
    ))
   &DynamicMaterialParameters(
    Index 0 Param 1=&FloatFromCurve(FloatCurve=DataInterface() {propertyValues {Curve {keys [
         {arriveTangent=1.2629, interpMode=RCIM_Cubic, leaveTangent=1.2629, tangentMode=RCTM_User, time=-0.0045, value=-1.9799}
         {arriveTangent=4.4591, interpMode=RCIM_Cubic, leaveTangent=4.4591, tangentMode=RCTM_User, time=1, value=0}
        ]}}})
    Index 0 Param 2=&FloatFromCurve(FloatCurve=DataInterface() {propertyValues {Curve {keys [<at 0.0102 4>,<at 0.9749 0>]}}})
    Index 0 Param 3=&MakeFloatFromVector(Vector=$Particles.Scale, Channel=::NewEnumerator0)
    Index 0 Param 4=1
    Index 1 Param 1=&FloatFromCurve(FloatCurve=DataInterface() {propertyValues {Curve {keys [<at 0.7 0.15 RCIM_Cubic>, {arriveTangent=6.4641, interpMode=RCIM_Cubic, leaveTangent=6.4641, tangentMode=RCTM_User, time=1, value=1}]}}})
    Module.Param0WriteEnabled=true
    Module.Param0WriteEnabled1=true
    Module.Param1WriteEnabled=true
    Module.Param1WriteEnabled1=true
    Module.Param2WriteEnabled=true
    Module.Param2WriteEnabled1=true
    Module.Param3WriteEnabled=true
    Module.Param3WriteEnabled1=true
    Write Parameter Index 1=true
   ) {module=DynamicMaterialParameters001}
  ]
  rendererClasses [&NiagaraMeshRendererProperties]
  renderers [{rendererClass=&NiagaraMeshRendererProperties, rendererIndex=0}]
 }
 Water_Plasma {
  bEnabled=false
  emitterUpdateScript [
   &EmitterState(Inactive Response=::NewEnumerator1, Life Cycle Mode=::NewEnumerator1, Loop Duration=1, UseLoopDelay=false)
   &SpawnBurst_Instantaneous(Use Spawn Probability=false, UseLoopCountLimit=false)
   &SpawnRate(SpawnRate=2, Use Spawn Probability=false)
  ]
  particleSpawnScript [
   &InitializeParticle(
    Lifetime=&RandomRangeFloat(Minimum=1.5, Maximum=2)
    Mesh Scale=<0.5 0.25 0.3>
    Mesh Scale Mode=::NewEnumerator4
    UsePositionOffset=true
    Write Material Random=true
   )
   &FN_SetSpawnGroup(Number of Spawn Groups=2)
   &SkeletalMeshLocation(
    Bone / Socket Index=2
    Bone Sampling Mode=::NewEnumerator4
    MaskSpawnGroup=false
    Orientation Sampling=::NewEnumerator1
    Position Sampling=::NewEnumerator2
    Sampled Position Offset Coordinate Space:ENiagaraSkelMeshTransforms=::NewEnumerator2
    Skeletal Mesh=$System.Skeletal Mesh
   ) {module=SkeletalMeshLocation001}
   &SkeletalMeshLocation(
    Bone / Socket Index=$Particles.SpawnGroup
    MaskSpawnGroup=false
    Mesh Sampling Type=::NewEnumerator5
    Sampled Position Offset=<0 -30 -10>
    Skeletal Mesh=$System.Skeletal Mesh
    Socket Sampling Mode=::NewEnumerator5
   ) {module=SkeletalMeshLocation002}
   &DynamicMaterialParameters(
    Index 0 Param 1=&RandomRangeFloat(Minimum=0, Maximum=1)
    Index 0 Param 3=1
    Module.Index0Param0Write=true
    Module.Index0Param1Write=true
    Module.Index0Param2Write=true
    Module.Index0Param3Write=false
   )
  ]
  particleUpdateScript [
   &ParticleState()
   &SkeletalMeshLocation(
    Bone / Socket Index=$Particles.SpawnGroup
    MaskSpawnGroup=false
    Mesh Sampling Type=::NewEnumerator5
    Sampled Position Offset=<0 0 -5>
    Skeletal Mesh=$System.Skeletal Mesh
    Socket Sampling Mode=::NewEnumerator5
   )
   /SetVariables(Particles.Lifetime=&MakeCustomFloatFromBool(
     Bool=&SetBoolByFloatComparison(
      A=30
      B=&DistanceBetweenPositions(
       A=&MakeVector(
        X=0
        Y=0
        Z=&MakeFloatFromVector(Vector=Linked() {linkedVariable {name=Particles.SkeletalMeshLocation001.SampledPosition, type=NiagaraPosition}}, Channel=::NewEnumerator2)
       )
       B=&MakeVector(
        X=0
        Y=0
        Z=&MakeFloatFromVector(Vector=Linked() {linkedVariable {name=Particles.SkeletalMeshLocation002.SampledPosition, type=NiagaraPosition}}, Channel=::NewEnumerator2)
       )
      )
     )
     True Float=$Particles.Lifetime
     False Float=0
    )) {isSetParameters=true}
   &ScaleColor(
    Scale Alpha=&FloatFromCurve(FloatCurve=DataInterface() {propertyValues {Curve {keys [
         {interpMode=RCIM_Cubic, tangentMode=RCTM_User, time=0, value=0}
         {interpMode=RCIM_Cubic, tangentMode=RCTM_User, time=0.5, value=1}
         {tangentMode=RCTM_User, time=1, value=0}
        ]}}})
    Scale RGB=&VectorFromFloat(Value=&FNTimeOfDayScalar(Night Scale=0.5, Morning Scale=1, Evening Scale=1))
    ScaleA=true
    ScaleRGB=true
   )
   &InterpolateOverTime(
    Rate Of Change=8
    Target Vector Value:Vector3f=&AddVectorToPosition(
     Position=$Particles.Position
     Vector=&RotateVectorByQuaternion(Vector To Rotate=<0 -100 0>, Quaternion=$Particles.SkeletalMeshLocation.SampledRotation)
    )
    Variable Type=::NewEnumerator2
   )
   &DynamicMaterialParameters(
    Index 2 Param 1=&MakeFloatFromVector(
     Vector=&Subtract_Position(
      A=&AddVectorToPosition(
       Position=$Particles.Position
       Vector=&RotateVectorByQuaternion(Vector To Rotate=<0 -100 0>, Quaternion=$Particles.SkeletalMeshLocation.SampledRotation)
      )
      B=Linked() {linkedVariable {name=Particles.InterpolateOverTime.InterpolateOverTime_V3_.Moving Average, type=Vector3f}}
     )
     Channel=::NewEnumerator0
    )
    Index 2 Param 2=&MakeFloatFromVector(
     Vector=&Subtract_Position(
      A=&AddVectorToPosition(
       Position=$Particles.Position
       Vector=&RotateVectorByQuaternion(Vector To Rotate=<0 -100 0>, Quaternion=$Particles.SkeletalMeshLocation.SampledRotation)
      )
      B=Linked() {linkedVariable {name=Particles.InterpolateOverTime.InterpolateOverTime_V3_.Moving Average, type=Vector3f}}
     )
     Channel=::NewEnumerator1
    )
    Index 2 Param 3=&MakeFloatFromVector(
     Vector=&Subtract_Position(
      A=&AddVectorToPosition(
       Position=$Particles.Position
       Vector=&RotateVectorByQuaternion(Vector To Rotate=<0 -100 0>, Quaternion=$Particles.SkeletalMeshLocation.SampledRotation)
      )
      B=Linked() {linkedVariable {name=Particles.InterpolateOverTime.InterpolateOverTime_V3_.Moving Average, type=Vector3f}}
     )
     Channel=::NewEnumerator2
    )
    Module.Param0WriteEnabled2=true
    Module.Param1WriteEnabled2=true
    Module.Param2WriteEnabled2=true
    Module.Param3WriteEnabled2=true
    Write Parameter Index 0=false
    Write Parameter Index 2=true
   ) {module=DynamicMaterialParameters001}
   &ScaleMeshSize(Scale Factor=&VectorFromCurve(
     VectorCurve=DataInterface() {propertyValues {
       XCurve {keys [{arriveTangent=1.898, interpMode=RCIM_Cubic, leaveTangent=1.898, tangentMode=RCTM_User, time=0, value=1}]}
       YCurve {keys [{arriveTangent=0.5312, interpMode=RCIM_Cubic, leaveTangent=0.5312, tangentMode=RCTM_User, time=0, value=1}, <at 1 1.2>]}
       ZCurve {keys [{arriveTangent=9.0725, interpMode=RCIM_Cubic, leaveTangent=9.0725, tangentMode=RCTM_User, time=0, value=1}]}
      }}
     Scale Curve=<1 1 1.5>
    )) {module=ScaleMeshSize001}
  ]
  rendererClasses [&NiagaraMeshRendererProperties]
  renderers [{rendererClass=&NiagaraMeshRendererProperties, rendererIndex=0}]
 }
 Water_Plasma_Splash001 {
  emitterUpdateScript [
   &EmitterState(
    Life Cycle Mode=::NewEnumerator1
    Loop Behavior=::NewEnumerator1
    Loop Delay=$System.Delay
    Loop Duration Mode=::NewEnumerator1
    UseLoopDelay=false
   )
   &SpawnRate(
    SpawnRate=&FloatFromCurve(
     FloatCurve=DataInterface() {propertyValues {Curve {keys [{interpMode=RCIM_Cubic, tangentMode=RCTM_User, time=0, value=1}, {interpMode=RCIM_Cubic, tangentMode=RCTM_User, time=1, value=0}]}}}
     CurveIndex=$Emitter.NormalizedLoopAge
     Scale Curve=10
    )
    Use Spawn Probability=false
   )
   &SpawnRate(
    SpawnGroup=1
    SpawnRate=&FloatFromCurve(
     FloatCurve=DataInterface() {propertyValues {Curve {keys [{interpMode=RCIM_Cubic, tangentMode=RCTM_User, time=0, value=1}, {interpMode=RCIM_Cubic, tangentMode=RCTM_User, time=1, value=0}]}}}
     CurveIndex=$Emitter.NormalizedLoopAge
     Scale Curve=10
    )
    Use Spawn Probability=false
   ) {module=SpawnRate001}
  ]
  particleSpawnScript [
   &FN_SetSpawnGroup(Number of Spawn Groups=2) {enabled=false}
   &InitializeParticle(
    Color Maximum=<rgba 0.7578 0.7578 0.9304 1>
    Color Minimum=<rgba 0.687 0.687 0.8435 1>
    Color Mode=::NewEnumerator2
    Lifetime=&RandomRangeFloat(Minimum=0.2, Maximum=0.25)
    Mesh Scale Max=<0.2 0.5 0.2>
    Mesh Scale Min=<0.15 0.4 0.1>
    Mesh Scale Mode=::NewEnumerator2
    Position Offset=&MakeVector(X=0, Y=0, Z=0)
    UsePositionOffset=true
    Write Material Random=true
   )
   &DynamicMaterialParameters(
    Index 0 Param 1=&RandomRangeFloat(Minimum=0, Maximum=1)
    Module.Index0Param0Write=true
    Module.Index0Param1Write=true
    Module.Index0Param2Write=true
    Module.Index0Param3Write=true
   )
   &AddVelocity(
    Module.Use Falloff Distance=false
    Origin Offset=<0 0 -2>
    Override Seed=false
    Velocity Mode=::NewEnumerator1
    Velocity Speed=&RandomRangeFloat(Minimum=10, Maximum=20)
   ) {enabled=false, module=AddVelocity001}
   &SkeletalMeshLocation(
    Bone / Socket Index=2
    Bone Sampling Mode=::NewEnumerator4
    MaskSpawnGroup=false
    Orientation Sampling=::NewEnumerator1
    Position Sampling=::NewEnumerator2
    Sampled Position Offset Coordinate Space:ENiagaraSkelMeshTransforms=::NewEnumerator2
    Skeletal Mesh=$System.Skeletal Mesh
   ) {module=SkeletalMeshLocation001}
   &SkeletalMeshLocation(
    Bone / Socket Index=$Particles.SpawnGroup
    MaskSpawnGroup=false
    Mesh Sampling Type=::NewEnumerator5
    Sampled Position Offset=&MakeVector(X=2, Y=&MaskFloatBySpawnGroup(Masked Float=5, Passthrough Float=-5), Z=0)
    Skeletal Mesh=$System.Skeletal Mesh
    Socket Sampling Mode=::NewEnumerator5
   )
   /SetVariables(Particles.Lifetime=&MakeCustomFloatFromBool(
     Bool=&SetBoolByFloatComparison(
      A=10
      B=&DistanceBetweenPositions(
       A=&MakeVector(
        X=0
        Y=0
        Z=&MakeFloatFromVector(Vector=Linked() {linkedVariable {name=Particles.SkeletalMeshLocation001.SampledPosition, type=NiagaraPosition}}, Channel=::NewEnumerator2)
       )
       B=&MakeVector(
        X=0
        Y=0
        Z=&MakeFloatFromVector(Vector=Linked() {linkedVariable {name=Particles.SkeletalMeshLocation.SampledPosition, type=NiagaraPosition}}, Channel=::NewEnumerator2)
       )
      )
     )
     True Float=$Particles.Lifetime
     False Float=0
    )) {isSetParameters=true}
   &InitialMeshOrientation(Mesh Orientation Mode=::NewEnumerator7, Use Rotation Vector=true) {enabled=false}
   &AddVelocity(
    Velocity=&ReturnMeshOrientationAxis(Mesh Orientation Axis=::XAxis, Mesh Orientation Quat=$Particles.MeshOrientation)
    Velocity Speed Scale=&RandomRangeFloat(Minimum=200, Maximum=300)
   ) {module=AddVelocity002}
   /SetVariables(Particles.Scale=&Multiply_VectorByFloat(
     Vector=$Particles.Scale
     Float=&FloatFromCurve(
      FloatCurve=DataInterface() {propertyValues {Curve {keys [<at 0 0>,<at 0.5 0>,<at 1 1>]}}}
      CurveIndex=&NormalizeDistanceRange(
       Start Position:NiagaraPosition=&MakeVector(
        X=0
        Y=&MakeFloatFromVector(Vector=Linked() {linkedVariable {name=Particles.SkeletalMeshLocation.SampledPosition, type=NiagaraPosition}}, Channel=::NewEnumerator1)
        Z=0
       )
       End Position:NiagaraPosition=&MakeVector(
        X=0
        Y=&MakeFloatFromVector(Vector=Linked() {linkedVariable {name=Particles.SkeletalMeshLocation001.SampledPosition, type=NiagaraPosition}}, Channel=::NewEnumerator1)
        Z=0
       )
       Distance:NiagaraFloat=50
       Invert Normalized Range:NiagaraBool=false
      )
     )
    )) {enabled=false, isSetParameters=true}
  ]
  particleUpdateScript [
   &ParticleState()
   &GravityForce(Gravity=<0 0 -10>)
   &SkeletalMeshLocation(
    Bone / Socket Index=$Particles.SpawnGroup
    MaskSpawnGroup=false
    Mesh Sampling Type=::NewEnumerator5
    Sampled Position Offset=&MakeVector(X=-20, Y=&MaskFloatBySpawnGroup(Masked Float=5, Passthrough Float=-8), Z=0)
    Skeletal Mesh=$System.Skeletal Mesh
    Socket Sampling Mode=::NewEnumerator5
   ) {enabled=false, module=SkeletalMeshLocation002}
   &Drag(Drag=2, Use Linear Drag=true, Use Rotational Drag=true)
   &ScaleColor(
    Scale RGB=&MakeVectorFromLinearColorRGB(Color=&ColorFromCurve(
      ColorCurve=DataInterface() {propertyValues {
        AlphaCurve {keys [<at 0 1 RCIM_Cubic>]}
        BlueCurve {keys [
          <at 0.0043 0.7>
          {arriveTangent=-0.2478, interpMode=RCIM_Cubic, leaveTangent=-0.2478, time=0.1789, value=0.7}
          <at 0.2812 0.6314>
          {arriveTangent=1.3318, interpMode=RCIM_Cubic, leaveTangent=1.3318, time=0.3527, value=1}
          {arriveTangent=-1.0217, interpMode=RCIM_Cubic, leaveTangent=-1.0217, time=0.558, value=1}
          {arriveTangent=-0.9848, interpMode=RCIM_Cubic, leaveTangent=-0.9848, time=0.8029, value=0.54}
          <at 0.8615 0.7011>
          <at 1 0.7>
         ]}
        GreenCurve {keys [
          <at 0.0043 0.35>
          {arriveTangent=1.1435, interpMode=RCIM_Cubic, leaveTangent=1.1435, time=0.1789, value=0.5785}
          <at 0.2812 0.6667>
          {arriveTangent=1.2043, interpMode=RCIM_Cubic, leaveTangent=1.2043, time=0.3527, value=1}
          {arriveTangent=-0.8536, interpMode=RCIM_Cubic, leaveTangent=-0.8536, time=0.558, value=1}
          {arriveTangent=-1.3917, interpMode=RCIM_Cubic, leaveTangent=-1.3917, time=0.8029, value=0.6157}
          <at 0.8615 0.5776>
          <at 1 0.35>
         ]}
        RedCurve {keys [
          <at 0.0043 0.28>
          {arriveTangent=2.246, interpMode=RCIM_Cubic, leaveTangent=2.246, time=0.1789, value=0.5542}
          <at 0.2812 0.902>
          {arriveTangent=0.3542, interpMode=RCIM_Cubic, leaveTangent=0.3542, time=0.3527, value=1}
          {arriveTangent=-0.2221, interpMode=RCIM_Cubic, leaveTangent=-0.2221, time=0.558, value=1}
          {arriveTangent=-1.476, interpMode=RCIM_Cubic, leaveTangent=-1.476, time=0.8029, value=0.9}
          <at 0.8615 0.552>
          <at 1 0.28>
         ]}
       }}
      CurveIndex=Linked() {linkedVariable {name=NPC.FortniteNPC.RemappedTimeOfDay_0-1, type=NiagaraFloat}}
     ))
    ScaleA=false
    ScaleRGB=true
   ) {module=ScaleColor001}
   &InterpolateOverTime(
    Rate Of Change=10
    Target Vector Value:Vector3f=&AddVectorToPosition(
     Position=$Particles.Position
     Vector=&RotateVectorByQuaternion(Vector To Rotate=<0 -200 0>, Quaternion=$Particles.SkeletalMeshLocation.SampledRotation)
    )
    Variable Type=::NewEnumerator2
   ) {enabled=false}
   &DynamicMaterialParameters(
    Index 0 Param 1=&FloatFromCurve(FloatCurve=DataInterface() {propertyValues {Curve {defaultValue=1, keys [{tangentMode=RCTM_User, time=0, value=0}, <at 0.5509 0.65>, {tangentMode=RCTM_User, time=1, value=1}]}}})
    Index 0 Param 3=0.5393
    Index 2 Param 1=&RandomRangeFloat(Minimum=1, Maximum=1.5)
    Index 2 Param 2=&RandomRangeFloat(Minimum=1, Maximum=1.5)
    Index 2 Param 3=-3.5
    Index 2 Param 4=&RandomRangeFloat(Minimum=3, Maximum=4)
    Index 3 Param 1=&RandomRangeFloat(Minimum=0.2, Maximum=0.3)
    Index 3 Param 2=&RandomRangeFloat(Minimum=0, Maximum=1)
    Index 3 Param 3=&RandomRangeFloat(Minimum=0, Maximum=1)
    Module.Index0Float0Write=true
    Module.Index0Float1Write=true
    Module.Index0Float2Write=true
    Module.Index0Float3Write=true
    Module.Index1Float0Write=true
    Module.Index1Float1Write=true
    Module.Index1Float2Write=true
    Module.Index1Float3Write=true
    Module.Index2Float0Write=true
    Module.Index2Float1Write=true
    Module.Index2Float2Write=true
    Module.Index2Float3Write=true
    Module.Index3Float0Write=true
    Module.Index3Float1Write=true
    Module.Index3Float2Write=true
    Module.Index3Float3Write=true
    Write Parameter Index 1=::NewEnumerator1
    Write Parameter Index 2=::NewEnumerator1
    Write Parameter Index 3=::NewEnumerator1
   ) {module=DynamicMaterialParameters001}
   &ScaleMeshSize(Scale Factor=&VectorFromCurve(
     VectorCurve=DataInterface() {propertyValues {
       XCurve {keys [<at 0.0084 0.4>,<at 0.496 0.8157>,<at 1 1>]}
       YCurve {keys [<at 0.0084 0.4>,<at 0.487 0.8157>,<at 1 1>]}
       ZCurve {keys [<at 0.0084 0.4>,<at 0.4825 0.8628>,<at 1 1>]}
      }}
     Scale Curve=&MaskVectorBySpawnGroup(Masked Vector=<-1 1 1>, Passthrough Vector=<1 1 1>)
    ))
   &SolveForcesAndVelocity(Clamp Velocity=false, Limit Acceleration=false)
  ]
  rendererClasses [&NiagaraMeshRendererProperties]
  renderers [{rendererClass=&NiagaraMeshRendererProperties, rendererIndex=0}]
 }
}
summary {
 systemName=NS_Emote_CoolSlice_SnowLooping_air
 emitters [
  {bEnabled=false, emitterName=DustSpikes, rendererClasses [&NiagaraMeshRendererProperties]}
  {bEnabled=false, emitterName=SmokeBack, rendererClasses [&NiagaraSpriteRendererProperties]}
  {bEnabled=false, emitterName=Flakes, simTarget=GPUComputeSim, rendererClasses [&NiagaraSpriteRendererProperties]}
  {bEnabled=false, emitterName=SmokeSides, rendererClasses [&NiagaraSpriteRendererProperties]}
  {emitterName=Clumps, rendererClasses [&NiagaraSpriteRendererProperties]}
  {bEnabled=false, emitterName=FlakesSide, simTarget=GPUComputeSim, rendererClasses [&NiagaraSpriteRendererProperties]}
  {bEnabled=false, emitterName=DustSpikesSide, rendererClasses [&NiagaraMeshRendererProperties]}
  {bEnabled=false, emitterName=Splash, rendererClasses [&NiagaraMeshRendererProperties]}
  {bEnabled=false, emitterName=SplashTop, rendererClasses [&NiagaraMeshRendererProperties]}
  {emitterName=Chunks, rendererClasses [&NiagaraMeshRendererProperties]}
  {bEnabled=false, emitterName=Goop, rendererClasses [&NiagaraRibbonRendererProperties]}
  {bEnabled=false, emitterName=Water_Plasma, rendererClasses [&NiagaraMeshRendererProperties]}
  {bEnabled=false, emitterName=Clumps001, rendererClasses [&NiagaraSpriteRendererProperties]}
  {bEnabled=false, emitterName=Chunks001, rendererClasses [&NiagaraMeshRendererProperties]}
  {emitterName=Water_Plasma_Splash001, rendererClasses [&NiagaraMeshRendererProperties]}
 ]
 userVariables [
  User.bIsInWater {type=NiagaraBool}
  User.Cuve Scale for Shading {type=NiagaraFloat, defaultValue=0.5, description=""}
  User.SmokeColor {
   type=NiagaraDataInterfaceColorCurve
   description=""
   defaultValue {
    struct=NiagaraExt_VariableValue_DataInterface
    value {
     dataInterface=&self:NiagaraDataInterfaceColorCurve_1
     dataInterfaceClass=&NiagaraDataInterfaceColorCurve
    }
   }
  }
  User.SnowColor {
   type=NiagaraDataInterfaceColorCurve
   description=""
   defaultValue {
    struct=NiagaraExt_VariableValue_DataInterface
    value {
     dataInterface=&self:NiagaraDataInterfaceColorCurve_0
     dataInterfaceClass=&NiagaraDataInterfaceColorCurve
    }
   }
  }
  User.SnowColorMax {type=Vector3f, description="", defaultValue {struct=Vector3f, value=<1 1 1>}}
  User.SnowColorMin {type=Vector3f, description="", defaultValue {struct=Vector3f, value=<0.5 0.8 1>}}
  User.VelocityMult {type=NiagaraFloat, defaultValue=2, description=""}
 ]
}
